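; Television prop Blueprint for Athena that manages layered broadcast audio attached to its audio location component.
; The warning-layer event spawns a warning sound, while the static-layer event stops that warning and has a one-in-ten random roll to start a static layer if none is playing.
; On death, it stops the static, static-layer and warning audio components; BeginPlay is empty, and both layer events have no caller here.
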(event EventBeginPlay)

(event Custom|Check%20to%20Spawn%20Warning%20Layer
  ;[1 1 1] Broadcast audio stuff here
  (bind _staticlayercomp (Variables|Default|GetStaticLayerComp)
    (:then
      (if (Audio|Components|Audio|IsPlaying _staticlayercomp)
        _
        (else
          _)))
    (:else
      _))
  (bind _returnvalue (Audio|SpawnSoundAttached :Sound (Variables|Default|GetStaticLayerWarningSound) :AttachToComponent (Variables|Default|GetAudioLoc) :AttachPointName "None" :LocationType "KeepRelativeOffset" :bStopWhenAttachedToDestroyed true :VolumeMultiplier 1.0 :PitchMultiplier 1.0 :StartTime 0.0 :bAutoDestroy true))
  (Variables|Default|SetWarningLayerComp _returnvalue))

(event Custom|Check%20to%20Spawn%20Static%20Layer
  ;[1 1 1] Broadcast audio stuff here
  (Utilities|FlowControl|Sequence
    (:then_0
      (bind _warninglayercomp (Variables|Default|GetWarningLayerComp)
        (:then
          (Audio|Components|Audio|Stop _warninglayercomp))
        (:else)))
    (:then_1
      (bind _staticlayercomp (Variables|Default|GetStaticLayerComp)
        (:then
          (if (Audio|Components|Audio|IsPlaying _staticlayercomp)
            _
            (elif (== (Math|Random|RandomIntegerinRange 1 10) 1)
              (bind _returnvalue (Audio|SpawnSoundAttached :Sound (Variables|Default|GetStaticLayerMSound) :AttachToComponent (Variables|Default|GetAudioLoc) :AttachPointName "None" :LocationType "KeepRelativeOffset" :bStopWhenAttachedToDestroyed true :VolumeMultiplier 1.0 :PitchMultiplier 1.0 :StartTime 0.0 :bAutoDestroy true))
              (Variables|Default|SetStaticLayerComp _returnvalue))))
        (:else
          _)))))

(event EventOnDeathPlayEffects
  (params (Damage :type Float%20%28single-precision%29) (DamageTags :type Gameplay%20Tag%20Container%20Structure%20%28by%20ref%29) (Momentum :type Vector :default "0, 0, 0") (HitInfo :type Hit%20Result%20Structure%20%28by%20ref%29) (InstigatedBy :type Fort%20Pawn%20Object%20Reference) (DamageCauser :type Actor%20Object%20Reference) (EffectContext :type Gameplay%20Effect%20Context%20Handle%20Structure))
  ;[1 1 1] Clean up audio on destroy
  (Utilities|FlowControl|Sequence
    (:then_0
      (bind _staticcomp (Variables|Default|GetStaticComp)
        (:then
          (Audio|Components|Audio|Stop _staticcomp))
        (:else)))
    (:then_1
      (bind _staticlayercomp (Variables|Default|GetStaticLayerComp)
        (:then
          (Audio|Components|Audio|Stop _staticlayercomp))
        (:else)))
    (:then_2
      (bind _warninglayercomp (Variables|Default|GetWarningLayerComp)
        (:then
          (Audio|Components|Audio|Stop _warninglayercomp))
        (:else)))))

(fn ConstructionScript)
