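; Parent Blueprint for Athena building container props, exposing time-of-day light and shadow defaults to its children.
; On cosmetic-capable instances, its construction script writes snow coverage to custom primitive data, caches light animation curve length and optionally clamps actor scale proportionally.
; A debug wind preview swaps the building mesh for a temporary copy whose dynamic materials blend toward intense-wind variants, with curve-driven speeds, yaw-based direction and warnings for missing setup.
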(fn GetTimeOfDayBlueprintDefaultVariables
  (return (Variables|TimeOfDayControlledLights|GetAlternateShadowStaticMesh) (Variables|TimeOfDayControlledLights|GetVolumetric%20Light%20Scattering%20Intensity) (Variables|TimeOfDayControlledLights|GetDisable%20TOD%20Lights%20and%20Material%20Emissive%20Values) (Variables|TimeOfDayControlledLights|GetDisable%20Static%20Mesh%20Shadow%20Casting%20When%20Lights%20Are%20Active) (Variables|TimeOfDayControlledLights|GetUse%20An%20Alternate%20Shadow%20Mesh%20When%20The%20Light%20is%20On%20) (Variables|TimeOfDayControlledLights|GetCast%20Volumetric%20Shadows)))

(fn ConstructionScript
  (bind _returnvalue (Transformation|GetActorScale3D))
  (bind _x (.x _returnvalue))
  (bind _y (.y _returnvalue))
  (bind _z (.z _returnvalue))
  (bind _returnvalue_2 (Math|Float|Max(Float) _x _y _z))
  (bind _max_scale (Variables|Default|GetMax%20Scale))
  (bind _returnvalue_3 (Wind|GetWindSpeedCurveforPreview))
  (bind _returnvalue_4 (Wind|GetWindPannerSpeedCurveforPreview))
  (bind _debug_wind_intensity (Variables|Wind|GetDebug%20Wind%20Intensity))
  (bind _returnvalue_5 (Math|Curves|GetLinearColorValue _returnvalue_3 _debug_wind_intensity))
  (bind _debug_wind_intensity_2 (Variables|Wind|GetDebug%20Wind%20Intensity))
  (bind _returnvalue_6 (Math|Curves|GetLinearColorValue _returnvalue_4 _debug_wind_intensity_2))
  (bind (_r _g _b _a) (Math|Color|BreakColor _returnvalue_5))
  (bind (_r_2 _g_2 _b_2 _a_2) (Math|Color|BreakColor _returnvalue_6))
  (bind _returnvalue_7 (Wind|GetWindPannerSpeedCurveforPreview))
  (Networking|CanExecuteCosmeticEvents
    (:True
      (Rendering|Material|SetCustomPrimitiveDataFloat (Variables|Debug|GetStaticMeshComponent) 6 (Variables|Default|GetSnow%20Coverage%20Amount))
      (Utilities|BuildingLightsPopulateLightAnimationCurveLength (Variables|TimeOfDayControlledLights|GetAnimate%20Lights%20With%20A%20Curve%20-%20Disables%20time%20of%20day%20light%20controls) (Variables|TimeOfDayControlledLights|GetAnimate%20Mesh%20MID%20Emissive%20Value%20with%20a%20Curve%20-%20Disables%20time%20of%20day%20light%20controls) (Variables|TimeOfDayControlledLights|GetLightAnimationCurve) (Variables|TimeOfDayControlledLights|GetCodeControlled_Animation%20Curve%20Animation%20Length))
      ; Set Mesh Scale
      (if (Variables|Default|GetSet%20Max%20Actor%20Scale)
        (if (> _returnvalue_2 _max_scale)
          (Transformation|SetActorScale3D (* (/ _returnvalue _returnvalue_2) _max_scale))))
      ; Wind
      (if (Variables|Wind|GetDebugWind)
        (Utilities|IsValid (Wind|GetWindSpeedCurveforPreview)
          (:Is%20Valid
            (Utilities|IsValid _returnvalue_7
              (:Is%20Valid
                (if (== (Utilities|Array|Length (Variables|Default|GetOriginalMaterials)) (Utilities|Array|Length (Variables|Default|GetIntenseWindMaterialsForPreview)))
                  (bind (_materials _returnvalue_8) (Wind|GetIntenseWindMaterialsforPreview))
                  (Variables|Default|SetIntenseWindMaterialsForPreview _materials)
                  (if _returnvalue_8
                    (bind _returnvalue_9 (AddComponent|NODE_AddStaticMeshComponent-0))
                    (bind _output_get (Variables|Default|SetWind%20Intensity%20Debug%20Mesh _returnvalue_9))
                    (Components|StaticMesh|SetStaticMesh _output_get (Variables|Mesh|GetStaticMesh))
                    (Rendering|SetVisibility (Building|GetBuildingMeshComponent) false false)
                    (bind _output_get_2 (Variables|Default|SetOriginalMaterials (Rendering|Material|GetMaterials _output_get)))
                    (for (_array_element _array_index) _output_get_2
                      (bind _returnvalue_10 (Rendering|Material|CreateDynamicMaterialInstance (Variables|Default|GetWind%20Intensity%20Debug%20Mesh) _array_index _array_element "WindDebugIntensityMID" :!class /Script/Engine.PrimitiveComponent))
                      (Variables|Default|SetDebug_TempMaterial _returnvalue_10)
                      (bind _asmaterial_instance (Utilities|Casting|CastToMaterialInstance (Utilities|Array|Get(acopy) (Variables|Default|GetIntenseWindMaterialsForPreview) _array_index)))
                      (Rendering|Material|InterpolateMaterialInstanceParameters (Variables|Default|GetDebug_TempMaterial) (Utilities|Casting|CastToMaterialInstance _array_element) _asmaterial_instance (Variables|Wind|GetDebug%20Wind%20Intensity))
                      (Rendering|Material|SetMaterial (Variables|Default|GetWind%20Intensity%20Debug%20Mesh) _array_index (Variables|Default|GetDebug_TempMaterial))
                      (Utilities|IsValid _returnvalue_3
                        (:Is%20Valid
                          (Rendering|Material|SetScalarParameterValue (Variables|Default|GetDebug_TempMaterial) "Wind_1_Speed_CodeReferenced" _r :!class /Script/Engine.MaterialInstanceDynamic)
                          (Rendering|Material|SetScalarParameterValue (Variables|Default|GetDebug_TempMaterial) "Wind_2_Speed_CodeReferenced" _g :!class /Script/Engine.MaterialInstanceDynamic)
                          (Rendering|Material|SetScalarParameterValue (Variables|Default|GetDebug_TempMaterial) "Wind_3_Speed_CodeReferenced" _b :!class /Script/Engine.MaterialInstanceDynamic)
                          (Rendering|Material|SetScalarParameterValue (Variables|Default|GetDebug_TempMaterial) "Wind_4_Speed_CodeReferenced" _a :!class /Script/Engine.MaterialInstanceDynamic)
                          (Rendering|Material|SetScalarParameterValue (Variables|Default|GetDebug_TempMaterial) "Wind_Panner_Speed_1x_CodeReferenced" _r_2 :!class /Script/Engine.MaterialInstanceDynamic)
                          (Rendering|Material|SetScalarParameterValue (Variables|Default|GetDebug_TempMaterial) "Wind_Panner_Speed_2x_CodeReferenced" _g_2 :!class /Script/Engine.MaterialInstanceDynamic)
                          (Rendering|Material|SetScalarParameterValue (Variables|Default|GetDebug_TempMaterial) "Wind_Panner_Speed_1y_CodeReferenced" _b_2 :!class /Script/Engine.MaterialInstanceDynamic)
                          (Rendering|Material|SetScalarParameterValue (Variables|Default|GetDebug_TempMaterial) "Wind_Panner_Speed_2y_CodeReferenced" _a_2 :!class /Script/Engine.MaterialInstanceDynamic)
                          (Rendering|Material|SetVectorParameterValue (Variables|Default|GetDebug_TempMaterial) "WindVector_CodeControlled" (Math|Color|MakeColor (.x (Math|Vector|RotateVectorAroundAxis (Math|Vector|MakeVector 1) (Variables|Wind|GetDebugWindYaw) "0, 0,1.000000")) (.y (Math|Vector|RotateVectorAroundAxis (Math|Vector|MakeVector 1) (Variables|Wind|GetDebugWindYaw) "0, 0,1.000000")) (.z (Math|Vector|RotateVectorAroundAxis (Math|Vector|MakeVector 1) (Variables|Wind|GetDebugWindYaw) "0, 0,1.000000"))) :!class /Script/Engine.MaterialInstanceDynamic)))))
                  (else
                    (Development|PrintString "The intense wind material list needs to be populated in:"))))
              (:Is%20Not%20Valid
                (Development|PrintString "Please add a wind speed curve reference to:" :!label _label))))
          (:Is%20Not%20Valid
            (Utilities|FlowControl|Sequence
              (:then_0
                (Development|PrintString "Please add a wind speed curve reference to:"))
              (:then_1
                (Utilities|IsValid _returnvalue_7
                  (:Is%20Not%20Valid
                    (goto _label)))))))
        (Development|PrintString (Utilities|GetDisplayName self))
        (else
          (Rendering|SetVisibility (Building|GetBuildingMeshComponent) true false))))))
