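; Parent Blueprint for the Athena tier 6 chest prop, handling its treasure light, smoke sheet, ambient sound, interaction and death rules.
; BeginPlay shows smoke and light; Destroyed hides both, fades the audio and flags the chest to be destroyed by player building placement.
; Interaction and death checks allow only the outpost owner for Outpost controllers, and anyone otherwise.
; Construction script removes the light through an effects quality check where cosmetics run; a debug cursor-over print remains.
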(event EventBeginPlay
  (Rendering|SetVisibility (Variables|Default|GetS_Chest_SmokeSheet) true false)
  (CallFunction|SetLightVisibility true))

(event MouseInput|EventActorBeginCursorOver
  (Development|PrintString "TEST!!!"))

(event EventDestroyed
  (Rendering|SetVisibility (Variables|Default|GetS_Chest_SmokeSheet) false false)
  (CallFunction|SetLightVisibility false)
  (Audio|Components|Audio|FadeOut (Variables|Default|GetChest_Ambient_Sound) 3 0.0 "Linear" :!class /Script/Engine.AudioComponent)
  (Variables|Building|SetbDestroyOnPlayerBuildingPlacement true) ; Allows players to build over the opened chest mesh (but not when there is loot in there)
)

(fn BlueprintCanInteract
  (params (InteractingPawn :type Fort%20Pawn%20Object%20Reference) (InteractionBeingAttempted :type EInteractionBeingAttempted%20Enum :default "FirstInteraction") (InteractionType :type TInteractionType%20Enum :default "IT_NoInteraction"))
  (bind _asfort_player_controller_outpost (Utilities|Casting|CastToFortPlayerControllerOutpost (Pawn|GetController InteractingPawn))
    (:then
      (return (Outpost|IsOutpostOwner _asfort_player_controller_outpost)))
    (:CastFailed
      (return true))))

(fn SetLightVisibility
  (params (Visible :type Boolean))
  (Utilities|IsValid (Variables|Default|GetTrasureLight)
    (:Is%20Valid
      (Rendering|SetVisibility (Variables|Default|GetTrasureLight) Visible false))))

(fn ShouldDie
  (params (Damage :type Float%20%28single-precision%29) (EventInstigator :type Controller%20Object%20Reference) (DamageCauser :type Actor%20Object%20Reference) (bIsTest :type Boolean) (EffectContext :type Gameplay%20Effect%20Context%20Handle%20Structure%20%28by%20ref%29))
  (bind _asfort_player_controller_outpost (Utilities|Casting|CastToFortPlayerControllerOutpost EventInstigator)
    (:then
      (return (Outpost|IsOutpostOwner _asfort_player_controller_outpost)))
    (:CastFailed
      (return true))))

(fn ConstructionScript
  (Networking|CanExecuteCosmeticEvents
    (:True
      (Utilities|EffectsQualityCheck 0
        (:Medium
          _)
        (:Low
          _))
      (Components|DestroyComponent (Variables|Default|GetTrasureLight)))))
