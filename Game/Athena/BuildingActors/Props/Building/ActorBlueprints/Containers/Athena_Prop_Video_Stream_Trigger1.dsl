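; In-world video screen prop that streams event video to clients, with picture-in-picture, subtitles, content blocking and CDN warming.
; It finds the BP_VideoMeshTrigger sharing its ScreenName, binds its events and replays Mesh variable state for join-in-progress clients; CMS events can also start playback.
; Commands pass an AllowCommand screen and region check.
; Timers use distance and line of sight to request URLs, drive playback and cap bandwidth, with retries on failed opens.
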(event Custom|OnStartPiP
  (params (ScreenName :type String) (VideoUID :type String) (Enable :type Boolean) (Live :type Boolean) (RegionData :type String))
  ;[1 1 1] Mesh Trigger
  ;;[0.1472 0.1238 1] Mesh Events
  (Mesh|UpdateMeshVariables)
  (Class|LiveStreamStandalonePlayer|SetPIPStartTime (Utilities|Struct|BreakMeshRegionLockDataFloat (Variables|MeshVars|GetMesh_PiPPlayerServerTimeSeconds)))
  (CallFunction|StartPiPStandalone ScreenName VideoUID Enable RegionData))

(event Custom|MeshTriggerReady
  (params (PlayStartTime :type Float%20%28double-precision%29) (VideoUID :type String) (ScreenName :type String) (Live :type Boolean) (RegionData :type String))
  ;[1 1 1] Mesh Trigger
  ;;[0.1472 0.1238 1] Mesh Events
  (if (Variables|Default|GetIsVideoTestEnabled)
    (CallFunction|Videotestoff (Variables|MeshVars|GetMesh_ScreenName) true RegionData))
  (CallFunction|PrePlayConfig)
  (Variables|MediaPlayer|SetVideoUID VideoUID)
  (Development|PrintString "StreamTrigger: readytoplay")
  (Variables|Default|SetLocalPlayUTCTime (Mesh|GetDataDateTime (Variables|MeshVars|GetMesh_RootUTCTIme)))
  (CallFunction|Videotest (Variables|MeshVars|GetMesh_ScreenName) (Variables|MediaPlayer|GetVideoUID) 0.0 RegionData))

(event Custom|MeshStartWarming
  (params (StartWarmingTime :type Float%20%28double-precision%29) (ScreenName :type String) (SimpleWarming :type Boolean) (RegionData :type String))
  ;[1 1 1] Mesh Trigger
  ;;[0.1472 0.1238 1] Mesh Events
  (Mesh|UpdateMeshVariables)
  (CallFunction|StartVideoWarming ScreenName true SimpleWarming (Mesh|GetDataString (Variables|MeshVars|GetMesh_WarmingCDNWeights)) RegionData))

(event Custom|MeshStopWarming
  (params (VideoScreen :type String) (SimpleWarming :type Boolean) (RegionData :type String))
  ;[1 1 1] Mesh Trigger
  ;;[0.1472 0.1238 1] Mesh Events
  (Mesh|UpdateMeshVariables)
  (CallFunction|StartVideoWarming VideoScreen false SimpleWarming (Mesh|GetDataString (Variables|MeshVars|GetMesh_WarmingCDNWeights)) RegionData))

(event Custom|OnScreenPrio
  (params (Distance :type Integer) (RegionData :type String))
  ;[1 1 1] Mesh Trigger
  ;;[0.1472 0.1238 1] Mesh Events
  (Mesh|UpdateMeshVariables))

(event Custom|MeshStopPlayback
  (params (VideoScreen :type String) (LeaveScreen :type Boolean) (RegionData :type String))
  ;[1 1 1] Mesh Trigger
  ;;[0.1472 0.1238 1] Mesh Events
  (if (Variables|Default|GetIsVideoTestEnabled)
    (Mesh|UpdateMeshVariables)
    (CallFunction|Videotestoff VideoScreen LeaveScreen RegionData)))

(event Custom|MeshSetMaxBW
  (params (ScreenName :type String) (MaxBW :type Integer) (RegionData :type String))
  ;[1 1 1] Mesh Trigger
  ;;[0.1472 0.1238 1] Mesh Events
  (Mesh|UpdateMeshVariables)
  (CallFunction|Videosetmaxbw ScreenName MaxBW RegionData))

(event Custom|MeshSeekPlayback
  (params (ScreenName :type String) (SeekTime :type Float%20%28double-precision%29) (Exact :type Boolean) (RegionData :type String))
  ;[1 1 1] Mesh Trigger
  ;;[0.1472 0.1238 1] Mesh Events
  (Mesh|UpdateMeshVariables)
  (CallFunction|Videoseek ScreenName SeekTime Exact RegionData))

(event Custom|MeshSetMirrorScreen
  (params (ScreenName :type String) (Enable :type Boolean) (RegionData :type String))
  ;[1 1 1] Mesh Trigger
  ;;[0.1472 0.1238 1] Mesh Events
  (Mesh|UpdateMeshVariables)
  (CallFunction|SetPiPMirror ScreenName Enable RegionData))

(event Custom|MeshCDNWeights
  (params (Weights :type String) (RegionData :type String))
  ;[1 1 1] Mesh Trigger
  ;;[0.1472 0.1238 1] Mesh Events
  (Mesh|UpdateMeshVariables)
  (CallFunction|SetCDNWeights Weights RegionData))

(event Custom|MeshForcePiPOn
  (params (Screen :type String) (Counter :type Integer64) (Regiondata :type String))
  ;[1 1 1] Mesh Trigger
  ;;[0.1472 0.1238 1] Mesh Events
  (Mesh|UpdateMeshVariables)
  (CallFunction|ForcePiPOn Screen Counter Regiondata))

(event Custom|MeshEnableLoopVideo
  (params (ScreenName :type String) (RegionData :type String))
  ;[1 1 1] Mesh Trigger
  ;;[0.1472 0.1238 1] Mesh Events
  (Mesh|UpdateMeshVariables)
  (CallFunction|EnableLoopVideo ScreenName RegionData))

(event Custom|MeshDisableLoopVideo
  (params (ScreenName :type String) (RegionData :type String))
  ;[1 1 1] Mesh Trigger
  ;;[0.1472 0.1238 1] Mesh Events
  (Mesh|UpdateMeshVariables)
  (CallFunction|DisableLoopVideo ScreenName RegionData))

(event Custom|SetupMeshTrigger
  (bind _mesh_playservertimeseconds (Variables|MeshVars|GetMesh_PlayServerTimeSeconds))
  (bind (_data _regiondata) (Utilities|Struct|BreakMeshRegionLockDataFloat _mesh_playservertimeseconds))
  (bind _mesh_pipenable (Variables|MeshVars|GetMesh_PiPEnable))
  (bind (_data_2 _regiondata_2) (Utilities|Struct|BreakMeshRegionLockDataBool _mesh_pipenable))
  (bind (_data_3 _regiondata_3) (Utilities|Struct|BreakMeshRegionLockDataString (Variables|MeshVars|GetMesh_WarmingCDNWeights)))
  (bind _mesh_pipforcecounter (Variables|MeshVars|GetMesh_PiPForceCounter))
  (bind (_data_4 _regiondata_4) (Utilities|Struct|BreakMeshRegionLockDataInt64 _mesh_pipforcecounter))
  (bind _mesh_pipmirrorenable (Variables|MeshVars|GetMesh_PiPMirrorEnable))
  (bind (_data_5 _regiondata_5) (Utilities|Struct|BreakMeshRegionLockDataBool _mesh_pipmirrorenable))
  ;[1 1 1] Mesh Trigger
  (bind _outactors (Actor|GetAllActorsOfClass "/Game/Movies/Season5/Blueprints/BP_VideoMeshTrigger.BP_VideoMeshTrigger_C" :!label _label_2))
  (bind _array_element (Utilities|Array|ForEachLoopwithBreak _outactors :!label _label)
    (:LoopBody
      (Utilities|IsValid _array_element
        (:Is%20Valid
          (if (Utilities|String|EqualExactly(String) (Variables|Default|GetScreenName) (Variables|/Game/Movies/Season5/Blueprints/BP_VideoMeshTrigger.BP_VideoMeshTrigger_C|GetScreenName _array_element))
            (Variables|MeshVars|SetMeshTrigger _array_element)
            (goto _label Break)))))
    (:Completed
      (Utilities|IsValid (Variables|MeshVars|GetMeshTrigger)
        (:Is%20Valid
          (bind _nodetype (Class|FortAsyncActionMeshNetworkReady|MeshNetworkReadyAsync)
            (:OnReady
              (switch Utilities|FlowControl|Switch|SwitchonEMeshNetworkNodeType _nodetype
                (:Client
                  (Utilities|FlowControl|Sequence
                    (:then_0
                      (Utilities|FlowControl|Sequence
                        (:then_0
                          (CallFunction|PrePlayConfig))
                        (:then_1
                          ;;[0.1472 0.1238 1] Mesh Events
                          (Default|BindEventtoOnReadytoPlay (Variables|MeshVars|GetMeshTrigger) (delegate Custom|MeshTriggerReady))
                          (Default|BindEventtoOnStopPlayback (Variables|MeshVars|GetMeshTrigger) (delegate Custom|MeshStopPlayback))
                          (Default|BindEventtoOnStartWarming (Variables|MeshVars|GetMeshTrigger) (delegate Custom|MeshStartWarming))
                          (Default|BindEventtoOnStopWarming (Variables|MeshVars|GetMeshTrigger) (delegate Custom|MeshStopWarming))
                          (Default|BindEventtoOnStartPiP (Variables|MeshVars|GetMeshTrigger) (delegate Custom|OnStartPiP))
                          (Default|BindEventtoOnScreenPrio (Variables|MeshVars|GetMeshTrigger) (delegate Custom|OnScreenPrio)))
                        (:then_2
                          (Default|BindEventtoOnEnableLoopVideo (Variables|MeshVars|GetMeshTrigger) (delegate Custom|MeshEnableLoopVideo))
                          (Default|BindEventtoOnDisableLoopVideo (Variables|MeshVars|GetMeshTrigger) (delegate Custom|MeshDisableLoopVideo))
                          (Default|BindEventtoOnSetBandwidthLimit (Variables|MeshVars|GetMeshTrigger) (delegate Custom|MeshSetMaxBW))
                          (Default|BindEventtoOnSeekPlayback (Variables|MeshVars|GetMeshTrigger) (delegate Custom|MeshSeekPlayback))
                          (Default|BindEventtoOnSetMirrorScreen (Variables|MeshVars|GetMeshTrigger) (delegate Custom|MeshSetMirrorScreen))
                          (Default|BindEventtoOnCDNWeights (Variables|MeshVars|GetMeshTrigger) (delegate Custom|MeshCDNWeights))
                          (Default|BindEventtoOnForcePiP (Variables|MeshVars|GetMeshTrigger) (delegate Custom|MeshForcePiPOn)))))
                    (:then_1
                      ;;[0.1506 1 0.2001] Mesh JIP Processing
                      (if (== _data -1.0)
                        (Development|PrintString (Utilities|String|Append "StreamTrigger: no play for " (Variables|Default|GetScreenName)) true true "(R=0.000000,G=0.660000,B=1.000000,A=1.000000)" 5.0)
                        (CallFunction|Videotestoff (Variables|MeshVars|GetMesh_ScreenName) (Mesh|GetDataBool (Variables|MeshVars|GetMesh_LeaveScreen)) _regiondata)
                        (else
                          (bind _isthisscreen (CallFunction|CheckScreen (Variables|MeshVars|GetMesh_ScreenName)))
                          (if _isthisscreen
                            (Development|PrintString "StreamTrigger: JIP Playback starting...")
                            (Variables|MediaPlayer|SetVideoUID (Mesh|GetDataString (Variables|MeshVars|GetMesh_VideoUID)))
                            (Variables|Default|SetLocalPlayUTCTime (Mesh|GetDataDateTime (Variables|MeshVars|GetMesh_PlayUTCTime)))
                            (CallFunction|Videotest (Variables|MeshVars|GetMesh_ScreenName) (Variables|MediaPlayer|GetVideoUID) 0.0 _regiondata)
                            (else
                              (Development|PrintString (Utilities|String|Append "StreamTrigger: no screen for " (Variables|Default|GetScreenName))))))))
                    (:then_2
                      (if (== (Utilities|Struct|BreakMeshRegionLockDataFloat (Variables|MeshVars|GetMesh_StartWarmingServerTimeSeconds)) -1.0)
                        (Development|PrintString (Utilities|String|Append "StreamTrigger: no warm for " (Variables|Default|GetScreenName)))
                        (elif (Variables|Default|GetIsActive)
                          (Development|PrintString "StreamTrigger: no warm because video is active")
                          (else
                            (CallFunction|StartVideoWarming (Variables|MeshVars|GetMesh_ScreenName) true (Mesh|GetDataBool (Variables|MeshVars|GetMesh_SimpleWarming)) _data_3 _regiondata_3)))))
                    (:then_3
                      (if (> _data_4 -1)
                        (CallFunction|ForcePiPOn (Variables|MeshVars|GetMesh_ScreenName) _data_4 _regiondata_4))
                      (if _data_5
                        (CallFunction|SetPiPMirror (Variables|MeshVars|GetMesh_ScreenName) true _regiondata_5))
                      (if _data_2
                        (CallFunction|StartPiPStandalone (Variables|MeshVars|GetMesh_ScreenName) (Mesh|GetDataString (Variables|MeshVars|GetMesh_PiPVideoUID)) true _regiondata_2)))))))))
        (:Is%20Not%20Valid
          (Utilities|FlowControl|Delay 1.0)
          (goto _label_2))))))

(event Custom|PrePlayConfig
  (bind (_data _regiondata) (Utilities|Struct|BreakMeshRegionLockDataString (Variables|MeshVars|GetMesh_CDNWeights)))
  (bind _mesh_loopvideo (Variables|MeshVars|GetMesh_LoopVideo))
  (bind (_data_2 _regiondata_2) (Utilities|Struct|BreakMeshRegionLockDataBool _mesh_loopvideo))
  (bind _mesh_maxbandwidth (Variables|MeshVars|GetMesh_MaxBandwidth))
  (bind (_data_3 _regiondata_3) (Utilities|Struct|BreakMeshRegionLockDataInt _mesh_maxbandwidth))
  (bind _mesh_screenname (Variables|MeshVars|GetMesh_ScreenName))
  ;[1 1 1] Mesh Trigger
  ;;[0.1506 1 0.2001] Mesh JIP Processing
  (Mesh|UpdateMeshVariables)
  (if _data_2
    (CallFunction|EnableLoopVideo (Variables|MeshVars|GetMesh_ScreenName) _regiondata_2))
  (CallFunction|SetCDNWeights _data _regiondata)
  (if (> _data_3 -1)
    (CallFunction|Videosetmaxbw _mesh_screenname _data_3 _regiondata_3)
    (else
      (CallFunction|Videosetmaxbw _mesh_screenname 0 "-")))
  (Variables|MediaPlayer|SetForceStartPlayTime false)
  (Variables|/Script/FortniteGame.FortStreamMediaSource|SetbBlurlLive (Mesh|GetDataBool (Variables|MeshVars|GetMesh_IsLive)) (Variables|MediaPlayer|GetMediaSource)))

(event EventBeginPlay
  ;[1 1 1] Begin play
  (if (Variables|Default|GetRemoteOnly)
    (Networking|SwitchHasAuthority
      (:Remote
        _)))
  (Utilities|FlowControl|Sequence
    (:then_0
      (if (Utilities|String|EqualExactly(String) (Variables|Default|GetScreenName) "RiskyReels")
        (Variables|Default|SetUseScreenMSC true))
      (if (Variables|Default|GetIsVideoTestEnabled)
        (Development|PrintString "stream trigger beginplay")))
    (:then_1
      (CallFunction|GetRegionInfo true))
    (:then_2
      (Components|Tick|SetComponentTickEnabled (Variables|Default|GetScreenMediaSound) false))
    (:then_3
      (CallFunction|SetupCMSEvents true))))

(event Custom|SetupCMSEvents
  (params (Setup :type Boolean))
  (bind _fortmediacmsevents (Variables|Default|GetFortMediaCMSEvents))
  ;[1 1 1] CMS Driven Video
  (Networking|CanExecuteCosmeticEvents
    (:True
      (if Setup
        (Variables|Default|SetVideoFromCMS false)
        (EventDispatchers|BindEventtoMediaStartedEvent _fortmediacmsevents (delegate Custom|CMSMediaStarted) :!class /Script/FortniteGame.FortMediaCMSEventsComponent)
        (EventDispatchers|BindEventtoMediaEndedEvent _fortmediacmsevents (delegate Custom|CMSMediaStopped))
        (Class|FortMediaCMSEventsComponent|InitializeEventPage _fortmediacmsevents "screens")))))

(event Custom|CMSMediaStarted
  (params (ValidEvent :type Media%20CMSEvent%20Object%20Reference))
  ;[1 1 1] CMS Driven Video
  (bind (_screen_name _prio _vuid _loop) (Utilities|CMSScreenCheck ValidEvent)
    (:This%20Screen
      (Variables|Default|SetVideoFromCMS true)
      (Variables|MeshVars|SetMesh_ScreenDistance (Utilities|Struct|MakeMeshRegionLockDataInt _prio "-"))
      (if _loop
        (Variables|Default|SetLoopVideoPlayerOption "Enabled")
        (else
          (Variables|Default|SetLoopVideoPlayerOption "Disabled")))
      (CallFunction|Videotest _screen_name _vuid -1.0 "-"))))

(event EventEndPlay
  (params (EndPlayReason :type EEndPlayReason%20Enum :default "Destroyed"))
  (bind _screenname (Variables|Default|GetScreenName))
  ;[1 1 1] End Play
  (if (Variables|Default|GetRemoteOnly)
    (Networking|SwitchHasAuthority
      (:Remote
        _)))
  (Media|MediaPlayer|Close (Variables|Default|GetMediaPlayerCtrl) :!class /Script/FortniteGame.FortMediaPlayerCtrl)
  (CallFunction|HandleMediaCtrl false)
  (ContentProtection|EnableShareBlocking false)
  (Screen|StopScreenTimers)
  (Class|FortMediaUCPInView|FireUCPInViewCallback false)
  (CallFunction|StopWarming)
  (Class|LiveStreamStandalonePlayer|UnblockPlayer "StreamInWorldActive")
  (if (Variables|PiP|GetPiPEnabled)
    (Variables|PiP|SetPiPForceOff true)
    (if (Variables|PiP|GetPiPStandAlone)
      (CallFunction|StartPiPStandalone _screenname "none" false "-")
      (else
        (CallFunction|SetPiPMirror _screenname false "-")))))

(event Custom|CMSMediaStopped
  (params (InvalidEvent :type Media%20CMSEvent%20Object%20Reference))
  ;[1 1 1] CMS Driven Video
  (bind _screen_name (Utilities|CMSScreenCheck InvalidEvent)
    (:This%20Screen
      (Variables|Default|SetVideoFromCMS false)
      (CallFunction|Videotestoff _screen_name false "-"))))

(event Custom|SetScreenActiveWithTimer
  ;[1 1 1] Set Screen Active w/ Timer
  (Variables|Default|SetIsActive true) ; The client has enough memory
  (Development|PrintString "Stream check periodically")
  (CallFunction|CheckDistancetoScreen)
  (bind _returnvalue (Utilities|Time|SetTimerbyEvent (EventDispatchers|CreateEvent :!event Custom|CheckDistanceToScreen) 0.001 true))
  (Variables|Default|SetCheckDistanceTimer _returnvalue)
  (bind _returnvalue_2 (Utilities|Time|SetTimerbyEvent (EventDispatchers|CreateEvent :!event Custom|BandwidthLimitDistance) 0.5 true))
  (Variables|Default|SetBandwidthCheckLimitTimer _returnvalue_2))

(event Custom|EnablePiPDistance
  (bind _respectscreenactivationrange (Variables|Default|GetRespectScreenActivationRange))
  (bind _mesh_screendistance (Variables|MeshVars|GetMesh_ScreenDistance))
  (bind _data (Mesh|GetDataInt _mesh_screendistance))
  (bind _activationcullradius_high (Variables|Default|GetActivationCullRadius_High))
  (bind _activationcullradius_low (Variables|Default|GetActivationCullRadius_Low))
  (bind _returnvalue (Utilities|Select :Index _data :Option%200 _activationcullradius_low :Option%201 _activationcullradius_high :Option%202 1.2e+05))
  (bind _returnvalue_2 (Math|Float|SelectFloat _returnvalue 1.2e+04 _respectscreenactivationrange))
  (bind _result (Default|ShouldStreambePlaying _returnvalue_2))
  (bind _mesh_pipversion (Variables|MeshVars|GetMesh_PIPVersion))
  (bind _data_2 (Mesh|GetDataInt _mesh_pipversion))
  (bind _returnvalue_3 (== _data_2 0))
  ;[1 1 1] Enable PiP on Distance from Screen
  (Utilities|PIPVersion
    (:New
      _)
    (:Old
      (if (not (Variables|PiP|GetPiPStandAlone))
        (PiP|SetPIPAnalyticsData false))))
  (bind _disabled (ContentProtection|CheckVideoDisabled _returnvalue_3))
  (ContentProtection|DrawPIPBlackout _returnvalue_3)
  (switch int (Mesh|GetDataInt (Variables|MeshVars|GetMesh_PIPVersion))
    (:-1
      _)
    (:0
      _)
    (:1
      (if (or _disabled _result)
        (PiP|RenderPIP false)
        (else
          (PiP|RenderPIP (Media|MediaPlayer|IsPlaying (Variables|MediaPlayer|GetMediaPlayer))))))
    (:2
      (if (or (and _result (Variables|Default|GetScreenLOS)) _disabled)
        (PiP|RenderPIP false)
        (else
          (PiP|RenderPIP (Variables|PiP|GetPiPEnabled))))))
  (if (or _disabled _result)
    (PiP|RenderPIP false)
    (else
      (PiP|RenderPIP true))))

(event Custom|CheckDistanceToScreen
  (bind _mesh_screendistance (Variables|MeshVars|GetMesh_ScreenDistance))
  (bind _data (Mesh|GetDataInt _mesh_screendistance))
  (bind _activationcullradius_high (Variables|Default|GetActivationCullRadius_High))
  (bind _activationcullradius_low (Variables|Default|GetActivationCullRadius_Low))
  (bind _returnvalue (Utilities|Select :Index _data :Option%200 _activationcullradius_low :Option%201 _activationcullradius_high :Option%202 1.2e+05))
  (bind _result (Default|ShouldStreambePlaying _returnvalue))
  ;[1 1 1] Check if stream should play
  (switch int (Mesh|GetDataInt (Variables|MeshVars|GetMesh_PIPVersion))
    (:-1
      _)
    (:0
      _)
    (:1
      _)
    (:2
      (bind _bounced (CallFunction|DebounceLOS))
      (if (not _bounced)
        (bind _los (Screen|CheckLOS false (not (Variables|MediaPlayer|GetMediaStarted_PIPLOS))))
        (if (Variables|MediaPlayer|GetMediaStarted_PIPLOS)
          (if (not (Math|Boolean|ANDBoolean _result (Variables|Default|GetIsVideoTestEnabled) _los))
            ;[1 1 1] Outside of streaming distance
            (if (Variables|Default|GetEnteredStreamingZone) :!label _label_2
              (Variables|Default|SetEnteredStreamingZone false)
              (CallFunction|SendRelevancyAnalytics)
              (if (Variables|Default|GetIsActive)
                _)
              (elif (and (Variables|MediaPlayer|GetMediaStarted_PIPLOS) (not (Variables|Default|GetIsVideoTestEnabled)))
                _))
            (switch int (Mesh|GetDataInt (Variables|MeshVars|GetMesh_PIPVersion))
              (:-1
                _)
              (:0
                _)
              (:1
                _)
              (:2
                (if (not (Variables|Default|GetIsVideoTestEnabled))
                  (Media|MediaPlayer|Close (Variables|Default|GetMediaPlayerCtrl) :!class /Script/FortniteGame.FortMediaPlayerCtrl :!label _label)
                  (CallFunction|HandleMediaCtrl false)
                  (Utilities|Platform|CollectGarbage)
                  (elif (Utilities|IsValid (Variables|Default|GetScreenActor) :!class /Script/Engine.KismetSystemLibrary)
                    (Default|CallVideoStoppedEventDispatcher)
                    (else
                      (Audio|ActivateTempMSC true))))))
            (Variables|MediaPlayer|SetMediaStarted_PIPLOS false)
            (goto _label))
          (elif (Variables|Default|GetIsVideoTestEnabled)
            _
            (else
              (goto _label_2))))
        ;[1 1 1] Inside Streaming Distance
        (if (not (Variables|Default|GetEnteredStreamingZone)) :!label _label_4
          (Variables|Default|SetEnteredStreamingZone true)
          (Variables|Default|SetTimeEnteredStreamingZone (Math|DateTime|UTCNow))
          (if (Variables|Default|GetIsActive)
            (if (not (Variables|MediaPlayer|GetVideoUIDSet))
              (Variables|MediaPlayer|SetVideoUID (Variables|MediaPlayer|GetVideoUID_Default)))
            (if (Variables|MediaPlayer|GetMediaStarted_PIPLOS)
              (if (Variables|Default|GetTempAudioUsed)
                (Audio|ActivateTempMSC false)
                (Default|CallAudioStartedEventDispatcher true true))
              (Default|CallVideoStartedEventDispatcher false)
              (else
                (Utilities|Platform|CollectGarbage)
                (switch int (Mesh|GetDataInt (Variables|MeshVars|GetMesh_PIPVersion))
                  (:-1
                    _)
                  (:0
                    _)
                  (:1
                    (if (Variables|Default|GetIsVideoTestEnabled) :!label _label_3
                      (Variables|MediaPlayer|SetMediaStarted_PIPLOS true)))
                  (:2
                    (goto _label_3)))
                (CallFunction|RequestVideoURL))))))))
  (if (and _result (Variables|Default|GetIsVideoTestEnabled))
    (goto _label_4)
    (else
      (goto _label_2))))

(event Custom|BandwidthLimitDistance
  ;[1 1 1] Limit Bandwidth on Distance from Screen
  (bind _disabled (ContentProtection|CheckVideoDisabled false))
  (if (not _disabled)
    (if (Default|ShouldStreambePlaying (Variables|Default|GetBandwidthLimitCullRadius))
      (Variables|/Script/FortniteGame.FortStreamMediaSource|SetMaxBandwidth (Mesh|GetDataInt (Variables|MeshVars|GetMesh_MaxBandwidth)) (Variables|MediaPlayer|GetMediaSource))
      (else
        (Variables|/Script/FortniteGame.FortStreamMediaSource|SetMaxBandwidth 1000000 (Variables|MediaPlayer|GetMediaSource))))))

(event Custom|RequestVideoURL
  (bind _mediasource (Variables|MediaPlayer|GetMediaSource))
  ;[1 1 1] Requesting Video URL
  (CallFunction|HandleMediaCtrl true)
  (CallFunction|BindMediaSourceEvents true)
  (Default|CallVideoAbouttoLoadEventDispatcher false)
  (Variables|/Script/FortniteGame.FortStreamMediaSource|SetStreamID (Variables|MediaPlayer|GetVideoUID) _mediasource)
  (Stream|RequestVideoUrl _mediasource (Fortnite|GetFirstLocalFortPlayerController true))
  (Development|PrintString "StreamTrigger: request url"))

(event Custom|videotest
  (params (ScreenName :type String) (VideoUID :type String) (StartTime :type Float%20%28double-precision%29) (RegionData :type String))
  ;[1 1 1] Video ScreenControl
  (bind _allowed (CallFunction|AllowCommand ScreenName RegionData))
  (if _allowed
    (Variables|MediaPlayer|SetDelayVideoStart false)
    (Variables|MediaPlayer|SetVideoUIDSet true)
    (Variables|MediaPlayer|SetVideoUID (Utilities|String|SelectString VideoUID (Variables|MediaPlayer|GetVideoUID_Default) (Math|Conversions|ToBoolean(Integer) (Utilities|String|Len VideoUID))))
    (if (not (== StartTime -1.0))
      (Variables|Default|SetStartPlayTime StartTime)
      (Variables|MediaPlayer|SetForceStartPlayTime true))
    (Development|PrintString "Stream video test")
    (Variables|Default|SetIsVideoTestEnabled true)
    (CallFunction|SetScreenActivewithTimer)))

(event Custom|On%20Video%20Url%20Success
  (params (RetrievedUrl :type String) (Code :type Integer) (RequestType :type EEpicMediaRequestType%20Enum :default "UNKNOWN"))
  ;[1 1 1] URL Request Success
  (Development|PrintString "Video URL Success")
  (bind _completeblock (ContentProtection|CheckBlocks))
  (if _completeblock
    (bind _returnvalue (Utilities|Time|SetTimerbyEvent (EventDispatchers|CreateEvent :!event Custom|CheckUCPFromStart) 0.05 true))
    (Variables|Default|SetCheckUCPFromStartTimer _returnvalue)
    (else
      (CallFunction|OnVideoUrlSuccesCont))))

(event Custom|CheckUCPFromStart
  ;[1 1 1] CheckUCPFromStart
  (bind _completeblock (ContentProtection|CheckBlocks))
  (if (not _completeblock)
    (Utilities|Time|ClearandInvalidateTimerbyHandle (Variables|Default|GetCheckUCPFromStartTimer))
    (CallFunction|OnVideoUrlSuccesCont)))

(event Custom|VideoPlayFrom
  (params (ScreenName :type String) (StartPlayTime :type Float%20%28double-precision%29) (RegionData :type String))
  ;[1 1 1] Video ScreenControl
  (bind _allowed (CallFunction|AllowCommand ScreenName RegionData))
  (if _allowed
    (Variables|Default|SetStartPlayTime StartPlayTime)
    (Variables|MediaPlayer|SetForceStartPlayTime true)
    (Development|PrintString "Stream video test")
    (Variables|Default|SetIsVideoTestEnabled true)
    (CallFunction|SetScreenActivewithTimer)))

(event Custom|StartPlayCheck
  ;[1 1 1] Call Video Started for Screen once actually playing
  (bind _returnvalue (Utilities|Time|SetTimerbyEvent (delegate Custom|PlayCheck) 0.1 true))
  (bind _output_get (Variables|Default|SetStartPlayingTimer _returnvalue)))

(event Custom|OnVideoUrlSucces_Cont
  (bind _mesh_playservertimeseconds (Variables|MeshVars|GetMesh_PlayServerTimeSeconds))
  (bind _data (Mesh|GetDataFloat _mesh_playservertimeseconds))
  ;[1 1 1] URL Request Success
  (CallFunction|BindMediaSourceEvents false)
  (CallFunction|StopWarming)
  (ContentProtection|UCPInViewTimer true)
  (if (Variables|/Script/FortniteGame.FortStreamMediaSource|GetbPartySync (Variables|MediaPlayer|GetMediaSource))
    (Media|MediaPlayer|SetPlayerStartTime (Variables|Default|GetMediaPlayerCtrl) _data)
    (Development|PrintString (Utilities|String|Append "StreamTrigger: Partysync Start Time: " (Utilities|String|ToString(Float) _data))))
  (CallFunction|AttemptOpenSource))

(event Custom|videotestoff
  (params (ScreenName :type String) (LeaveScreen :type Boolean) (RegionData :type String))
  ;[1 1 1] Video ScreenControl
  (bind _allowed (CallFunction|AllowCommand ScreenName RegionData))
  (if _allowed
    (Variables|Default|SetLeaveScreen LeaveScreen)
    (Development|PrintString "Stream video test off")
    (CallFunction|End)))

(event Custom|PlayCheck
  ;[1 1 1] Call Video Started for Screen once actually playing
  (if (Media|MediaPlayer|IsPlaying (Variables|MediaPlayer|GetMediaPlayer))
    (Utilities|Time|ClearandInvalidateTimerbyHandle (stale Custom|StartPlayCheck _output_get))
    (if (Default|ShouldStreambePlaying (Utilities|Select :Index (Mesh|GetDataInt (Variables|MeshVars|GetMesh_ScreenDistance)) :Option%200 (Variables|Default|GetActivationCullRadius_Low) :Option%201 (Variables|Default|GetActivationCullRadius_High) :Option%202 1.2e+05))
      (switch int (Mesh|GetDataInt (Variables|MeshVars|GetMesh_PIPVersion))
        (:-1
          _)
        (:0
          _)
        (:1
          _)
        (:2
          (if (Variables|Default|GetScreenLOS)
            _)))
      (Default|CallVideoStartedEventDispatcher true)
      (else
        (switch int (Mesh|GetDataInt (Variables|MeshVars|GetMesh_PIPVersion))
          (:1
            _)
          (:2
            _))
        (if (Utilities|IsValid (Variables|Default|GetScreenActor) :!class /Script/Engine.KismetSystemLibrary)
          (Default|CallAudioStartedEventDispatcher true true)
          (else
            (Audio|ActivateTempMSC true)))))))

(event Custom|videosetmaxbw
  (params (ScreenName :type String) (MaxBandwidth :type Integer) (RegionData :type String))
  ;[1 1 1] Video ScreenControl
  (bind _allowed (CallFunction|AllowCommand ScreenName RegionData))
  (if _allowed
    (Variables|/Script/FortniteGame.FortStreamMediaSource|SetMaxBandwidth MaxBandwidth (Variables|MediaPlayer|GetMediaSource))))

(event Custom|On%20Video%20Url%20Failed
  (params (RetrievedUrl :type String) (Code :type Integer) (RequestType :type EEpicMediaRequestType%20Enum :default "UNKNOWN"))
  ;[1 1 1] URL Request Failed
  (Development|PrintString "Video URL Failed")
  (CallFunction|TelevisionOff)
  (CallFunction|BindMediaSourceEvents false)
  (Default|CallVideoAbouttoLoadEventDispatcher true)
  (PiP|EndPlaybackClosePIP))

(event Custom|videoseek
  (params (ScreenName :type String) (SeekTime :type Float%20%28double-precision%29) (Exact :type Boolean) (RegionData :type String))
  (bind _mediaplayer (Variables|MediaPlayer|GetMediaPlayer))
  ;[1 1 1] Video ScreenControl
  (bind _allowed (CallFunction|AllowCommand ScreenName RegionData))
  (if _allowed
    (if (Media|MediaPlayer|SupportsSeeking _mediaplayer)
      (if Exact
        (Media|MediaPlayer|Seek _mediaplayer (Math|Timespan|FromSeconds SeekTime))
        (else
          (Media|MediaPlayer|Seek _mediaplayer (Math|Timespan|FromSeconds (+ SeekTime (Math|Timespan|GetTotalSeconds (Media|MediaPlayer|GetTime _mediaplayer))))))))))

(event Custom|On%20Media%20Opened
  (params (OpenedUrl :type String))
  (bind _subtitleplayer (Variables|Default|GetSubtitlePlayer))
  ;[1 1 1] Media Opened
  (Development|PrintString "stream media opened")
  (FortniteSubtitles|SubtitlesPlayer|SetSubtitles _subtitleplayer (Variables|Default|GetSubtitlesOverlays))
  (FortniteSubtitles|SubtitlesPlayer|Play _subtitleplayer)
  (Media|MediaPlayer|Play (Variables|Default|GetMediaPlayerCtrl) :!class /Script/FortniteGame.FortMediaPlayerCtrl)
  (if (Variables|Default|GetUseScreenMSC)
    (Audio|ActivateScreenMSC true))
  (CallFunction|StartPlayCheck)
  (ContentProtection|EnableShareBlocking true)
  (Audio|SetAudioSubmix))

(event Custom|SetCDNWeights
  (params (Weights :type String) (RegionData :type String))
  (bind _mesh_pipcdndistroset (Variables|MeshVars|GetMesh_PIPCDNDistroSet))
  (bind _data (Mesh|GetDataBool _mesh_pipcdndistroset))
  (bind _mediasource (Variables|MediaPlayer|GetMediaSource))
  (bind _cdnhostnames (Variables|/Script/FortniteGame.FortStreamMediaSource|GetCDNHostNames _mediasource))
  (bind _cdndistribution (Variables|/Script/FortniteGame.FortMediaCDNHostnames|GetCDNDistribution _cdnhostnames))
  ;[1 1 1] Video ScreenControl
  (bind _allowed (CallFunction|AllowCommand (Variables|MeshVars|GetMesh_ScreenName) RegionData))
  (if _allowed
    (if (not (Utilities|String|IsEmpty Weights))
      (Utilities|Array|Clear (Variables|/Script/FortniteGame.FortMediaCDNHostnames|GetCDNDistribution (Variables|/Script/FortniteGame.FortStreamMediaSource|GetCDNHostNames (Variables|MediaPlayer|GetMediaSource))))
      (for _array_element (Utilities|String|ParseIntoArray Weights " " false)
        (Utilities|Array|Add (Variables|/Script/FortniteGame.FortMediaCDNHostnames|GetCDNDistribution (Variables|/Script/FortniteGame.FortStreamMediaSource|GetCDNHostNames (Variables|MediaPlayer|GetMediaSource))) (Utilities|String|StringToFloat _array_element)))
      ;[1 0.1554 0.5754] PIP
      (if _data
        (Class|LiveStreamHUDElement|SetPIPCDNWeights _cdndistribution))
      (else
        (Utilities|Array|Clear (Variables|/Script/FortniteGame.FortMediaCDNHostnames|GetCDNDistribution (Variables|/Script/FortniteGame.FortStreamMediaSource|GetCDNHostNames (Variables|MediaPlayer|GetMediaSource))))
        (if _data
          (Class|LiveStreamHUDElement|SetPIPCDNWeights _cdndistribution))))))

(event Custom|On%20Media%20Closed
  ;[1 1 1] Media Closed
  (CallFunction|StreamFailCheck)
  (FortniteSubtitles|SubtitlesPlayer|Stop (Variables|Default|GetSubtitlePlayer))
  (if (Variables|MediaPlayer|GetVideoEnableLocalFallback)
    (CallFunction|AttemptOpenSource)
    (elif (Variables|MediaPlayer|GetStreamingFailed)
      (Utilities|FlowControl|Delay 1.0)
      (CallFunction|BindPlayerEvents false)
      (CallFunction|AttemptOpenSource)
      (else
        (CallFunction|TelevisionOff)
        (ContentProtection|EnableShareBlocking false)
        (Development|PrintString "stream media closed")))))

(event Custom|StartPiPStandalone
  (params (ScreenName :type String) (VideoUID :type String) (Enable :type Boolean) (RegionData :type String))
  (bind _returnvalue (Options|GetRuntimeOptions))
  ;[1 1 1] PiP Control
  (bind _allowed (CallFunction|AllowCommand ScreenName RegionData))
  (if _allowed
    (Utilities|PIPVersion
      (:New
        (PiP|ControlNewSAPIP Enable))
      (:Old
        (if Enable
          (Class|LiveStreamHUDElement|SetPIPStartTime (Mesh|GetDataFloat (Variables|MeshVars|GetMesh_PiPPlayerServerTimeSeconds)))
          (Class|LiveStreamHUDElement|SetPIPLive (Mesh|GetDataBool (Variables|MeshVars|GetMesh_IsPIPLive)))
          (Class|LiveStreamHUDElement|SetMirrorScreen -1)
          (Class|LiveStreamHUDElement|SetPIPVideoUID VideoUID)
          (Variables|PiP|SetPiPWasPlaying true)
          (Variables|PiP|SetPiPWasPlayingVUID VideoUID)
          (Variables|PiP|SetPiPEnabled true)
          (Variables|PiP|SetPiPStandAlone true)
          (Variables|/Script/FortniteGame.FortRuntimeOptions|SetbEnableLiveStream true _returnvalue)
          (CallFunction|EnablePiPDistanceCheck true)
          (else
            (Class|LiveStreamStandalonePlayer|ResetPIPState)
            (Variables|PiP|SetPiPWasPlayingVUID "")
            (Variables|PiP|SetPiPStandAlone false)
            (bind _disabled (PiP|CheckPiPDisabled))
            (if _disabled
              (Class|LiveStreamHUDElement|SetPIPEOSGSBlackout false)
              (Variables|PiP|SetPiPForceOff false)
              (Variables|PiP|SetPiPEnabled false)
              (Variables|/Script/FortniteGame.FortRuntimeOptions|SetbEnableLiveStream false _returnvalue)
              (CallFunction|EnablePiPDistanceCheck false)
              (if (Variables|PiP|GetPiPWasPlaying)
                (PiP|SetPIPAnalyticsData true)
                (Variables|PiP|SetPiPWasPlaying false)))))))))

(event Custom|On%20Media%20End
  ;[1 1 1] Media End
  (Development|PrintString "stream media end")
  (FortniteSubtitles|SubtitlesPlayer|Stop (Variables|Default|GetSubtitlePlayer))
  (PiP|EndPlaybackClosePIP)
  (CallFunction|TelevisionOff)
  (CallFunction|End)
  (Media|MediaPlayer|Close (Variables|Default|GetMediaPlayerCtrl) :!class /Script/FortniteGame.FortMediaPlayerCtrl)
  (CallFunction|HandleMediaCtrl false))

(event Custom|SetPiPMirror
  (params (ScreenName :type String) (Enable :type Boolean) (RegionData :type String))
  (bind _returnvalue (Options|GetRuntimeOptions))
  ;[1 1 1] PiP Control
  (bind _allowed (CallFunction|AllowCommand ScreenName RegionData))
  (if _allowed
    (Utilities|PIPVersion
      (:New
        (PiP|ControlNewMirrorPIP Enable))
      (:Old
        (if Enable
          (Class|LiveStreamHUDElement|SetMirrorScreen (Utilities|Array|FindItem (Variables|Default|GetScreenIndexes) ScreenName))
          (Variables|PiP|SetPiPEnabled true)
          (Variables|/Script/FortniteGame.FortRuntimeOptions|SetbEnableLiveStream true _returnvalue)
          (CallFunction|EnablePiPDistanceCheck true)
          (else
            (bind _disabled (PiP|CheckPiPDisabled))
            (if _disabled
              (Class|LiveStreamHUDElement|SetPIPEOSGSBlackout false)
              (Variables|PiP|SetPiPForceOff false)
              (Variables|PiP|SetPiPEnabled false)
              (Variables|/Script/FortniteGame.FortRuntimeOptions|SetbEnableLiveStream false _returnvalue)
              (CallFunction|EnablePiPDistanceCheck false)
              (PiP|SetPIPAnalyticsData true))))))))

(event Custom|OnMetaDataFailed
  (params (Reason :type String))
  ;[1 1 1] Turn of screen on a metadata error from the Media Source
  (Development|PrintString (Utilities|String|Append "Media Metadata Failed: " Reason))
  (PiP|EndPlaybackClosePIP)
  (CallFunction|TelevisionOff)
  (ContentProtection|EnableShareBlocking false))

(event Custom|StopWarming
  ;[1 1 1] PiP Control
  (CallFunction|StartVideoWarming (Variables|Default|GetScreenName) false (Mesh|GetDataBool (Variables|MeshVars|GetMesh_SimpleWarming)))
  (Development|PrintString "Stop Warming"))

(event Custom|EnablePiPDistanceCheck
  (params (Enable :type Boolean))
  ;[1 1 1] PiP Control
  (if Enable
    (bind _returnvalue (Utilities|Time|SetTimerbyEvent (EventDispatchers|CreateEvent :!event Custom|EnablePiPDistance) 0.05 true))
    (Variables|PiP|SetPiPActivation _returnvalue)
    (else
      (Utilities|Time|ClearandInvalidateTimerbyHandle (Variables|PiP|GetPiPActivation))
      (Utilities|PIPVersion
        (:Old
          (PiP|RenderPIP false))))))

(event Custom|ForcePiPOn
  (params (ScreenName :type String) (NewCounterValue :type Integer64) (RegionData :type String))
  ;[1 1 1] PiP Control
  (bind _allowed (CallFunction|AllowCommand ScreenName RegionData))
  (if (and _allowed (Math|Integer64|NotEqual(Integer64) NewCounterValue -1))
    (Variables|/Script/FortniteGame.FortRuntimeOptions|SetShouldShowLiveStreamPiPInMatchCounter NewCounterValue (Options|GetRuntimeOptions))))

(event Custom|On%20Media%20Open%20Failed
  (params (FailedUrl :type String))
  ;[1 1 1] Retry On Media Open Failed (uses same counter as closed)
  (if (< (Variables|Default|GetVideoRetryCounter) (Variables|Default|GetVideoMaxRetries))
    (CallFunction|BindPlayerEvents false)
    (Math|Integer|IncrementInt (Variables|Default|GetVideoRetryCounter)
      (:%20%20
        (Utilities|FlowControl|Delay 1.0)
        (CallFunction|RequestVideoURL)))
    (else
      (Variables|Default|SetVideoRetryCounter 0)
      (Screen|StopScreenTimers)
      (CallFunction|TelevisionOff)
      (PiP|EndPlaybackClosePIP))))

(event Custom|EnableLoopVideo
  (params (ScreenName :type String) (RegionData :type String))
  ;[1 1 1] Looping
  (bind _allowed (CallFunction|AllowCommand ScreenName RegionData))
  (if _allowed
    (Variables|Default|SetLoopVideoPlayerOption "Enabled")
    (if (Variables|Default|GetIsVideoTestEnabled)
      (Media|MediaPlayer|SetLooping (Variables|MediaPlayer|GetMediaPlayer) (Utilities|Enum|Equal(Enum) (Variables|Default|GetLoopVideoPlayerOption) "Enabled")))))

(event Custom|AttemptOpenSource
  (bind _returnvalue (Math|Timespan|FromSeconds 0.0))
  (bind _mediaplayeroptions (Utilities|Struct|MakeMediaPlayerOptions "(Audio=0,Caption=-1,Metadata=-1,Script=-1,Subtitle=-1,Text=-1,Video=0)" "(Video=\"\",Audio=\"\",Subtitle=\"\",Caption=\"\")" "UseTrackOptionIndices" _returnvalue "RelativeToStartTime" "UseMediaPlayerSetting" "Disabled"))
  (bind _loopvideoplayeroption (Variables|Default|GetLoopVideoPlayerOption))
  (bind _mediaplayeroptions_2 (Utilities|Struct|MakeMediaPlayerOptions "(Audio=0,Caption=-1,Metadata=-1,Script=-1,Subtitle=-1,Text=-1,Video=0)" "(Video=\"\",Audio=\"\",Subtitle=\"\",Caption=\"\")" "UseTrackOptionIndices" "+00000000.00:00:00.000000000" "RelativeToStartTime" "UseMediaPlayerSetting" _loopvideoplayeroption))
  (bind _mediasource (Variables|MediaPlayer|GetMediaSource))
  (bind _bislive (Variables|/Script/FortniteGame.FortStreamMediaSource|GetbIsLive _mediasource))
  (bind _returnvalue_2 (select _bislive _mediaplayeroptions _mediaplayeroptions_2))
  ;[1 1 1] Attempt Open Source
  (CallFunction|SetStartPlayTime)
  (if (not (Variables|MediaPlayer|GetDelayVideoStart)) :!label _label
    (CallFunction|BindPlayerEvents true)
    (Class|LiveStreamStandalonePlayer|BlockPlayer "StreamInWorldActive")
    (if (Variables|MediaPlayer|GetVideoEnableLocalFallback)
      (Media|MediaPlayer|OpenSourcewithOptions (Variables|MediaPlayer|GetMediaPlayer) (Variables|Default|GetFallbackMediaSource) _returnvalue_2 :!class /Script/MediaAssets.MediaPlayer)
      (else
        (Media|MediaPlayer|OpenSourcewithOptions (Variables|Default|GetMediaPlayerCtrl) (Variables|MediaPlayer|GetMediaSource) _returnvalue_2 :!class /Script/FortniteGame.FortMediaPlayerCtrl)))
    (FortniteSubtitles|SubtitlesPlayer|BindtoMediaPlayer (Variables|Default|GetSubtitlePlayer) (Variables|MediaPlayer|GetMediaPlayer))
    (Variables|MediaPlayer|SetVideoEnableLocalFallback false)
    (Variables|MediaPlayer|SetStreamingFailed false)
    (else
      (Utilities|FlowControl|Delay 0.05)
      (goto _label))))

(event Custom|DisableLoopVideo
  (params (ScreenName :type String) (RegionData :type String))
  ;[1 1 1] Looping
  (bind _allowed (CallFunction|AllowCommand ScreenName RegionData))
  (if _allowed
    (Variables|Default|SetLoopVideoPlayerOption "Disabled")
    (if (Variables|Default|GetIsVideoTestEnabled)
      (Media|MediaPlayer|SetLooping (Variables|MediaPlayer|GetMediaPlayer) (Utilities|Enum|Equal(Enum) (Variables|Default|GetLoopVideoPlayerOption) "Enabled")))))

(event Custom|StartVideoWarming
  (params (ScreenName :type String) (Enable :type Boolean) (Simple :type Boolean) (Weights :type String) (RegionData :type String))
  (bind _returnvalue (Transformation|GetActorTransform))
  ;[1 1 1] Video start warming
  (bind _allowed (CallFunction|AllowCommand ScreenName RegionData))
  (if _allowed
    (if Enable
      (Development|PrintString "StreamTrigger: video warm")
      (if Simple
        (Variables|Default|SetSimpleWarming true)
        (if (not (Utilities|IsValid (Variables|Default|GetNoMPMediaCDNWarmer) :!class /Script/Engine.KismetSystemLibrary))
          (bind _returnvalue_2 (Game|SpawnActorfromClass "/Game/Movies/MediaCDNWarmer.MediaCDNWarmer_C" _returnvalue "Undefined" "OverrideRootScale" 0 Weights 0))
          (Variables|Default|SetNoMPMediaCDNWarmer _returnvalue_2))
        (elif (Utilities|IsValid (Variables|Default|GetWarmingObject) :!class /Script/Engine.KismetSystemLibrary)
          _
          (else
            (bind _returnvalue_3 (Game|SpawnActorfromClass "/Game/Athena/Prototype/Blueprints/Galileo/Blueprints/BP_Galileo_VideoPreBuffer.BP_Galileo_VideoPreBuffer_C" _returnvalue "Undefined" "OverrideRootScale" 0 0))
            (Variables|Default|SetWarmingObject _returnvalue_3))))
      (else
        (Development|PrintString "StreamTrigger: video stop warm")
        (if (Variables|Default|GetSimpleWarming)
          (Actor|DestroyActor (Variables|Default|GetNoMPMediaCDNWarmer))
          (Variables|Default|SetSimpleWarming false)
          (else
            (Actor|DestroyActor (Variables|Default|GetWarmingObject))))))))

(fn ActivateScreenMSC
  (params (Enable :type Boolean))
  (bind _screenmediasound (Variables|Default|GetScreenMediaSound))
  (if Enable
    (Media|MediaSoundComponent|SetMediaPlayer _screenmediasound (Variables|MediaPlayer|GetMediaPlayer))
    (Components|Activation|Activate _screenmediasound false)
    (Synth|Components|Audio|Start _screenmediasound)
    (Audio|FlipAudioSubmix (not (or (Utilities|Enum|Equal(Enum) (Player|ShouldProtectPlayerfromContent (Variables|MediaPlayer|GetMediaSource)) "UCPAudio") (Utilities|Enum|Equal(Enum) (Player|ShouldProtectPlayerfromContent (Variables|MediaPlayer|GetMediaSource)) "UCPAudio"))))
    (else
      (Synth|Components|Audio|Stop _screenmediasound)
      (Components|Activation|Deactivate _screenmediasound))))

(fn ActivateTempMSC
  (params (Enable :type Boolean))
  (bind _output_get (Variables|Default|SetTempAudioUsed Enable))
  (Audio|Components|Audio|SetOutputtoBusOnly (Variables|Default|GetScreenMediaSound) _output_get :!class /Script/AudioMixer.SynthComponent))

(fn AllowCommand
  (params (ScreenName :type String) (RegionData :type String))
  (bind _notblocked (CallFunction|CheckBlacklist))
  (bind _isthisscreen (CallFunction|CheckScreen ScreenName))
  (if (and _isthisscreen _notblocked)
    (Mesh|UpdateMeshVariables)
    (bind _allowed (CallFunction|RegionIsAllowed RegionData))
    (return _allowed)
    (else
      (return false))))

(fn BindPlayerEvents
  (params (Bind :type Boolean))
  (bind _mediaplayer (Variables|MediaPlayer|GetMediaPlayer))
  (bind _mediaplayer_2 (Variables|MediaPlayer|GetMediaPlayer))
  (if Bind
    (Media|MediaPlayer|BindEventtoOnMediaOpenFailed _mediaplayer (EventDispatchers|CreateEvent :!event On%20Media%20Open%20Failed))
    (Media|MediaPlayer|BindEventtoOnMediaClosed _mediaplayer (EventDispatchers|CreateEvent :!event On%20Media%20Closed))
    (Media|MediaPlayer|BindEventtoOnMediaOpened _mediaplayer (EventDispatchers|CreateEvent :!event On%20Media%20Opened))
    (Media|MediaPlayer|BindEventtoOnEndReached _mediaplayer (EventDispatchers|CreateEvent :!event On%20Media%20End))
    (else
      (Media|MediaPlayer|UnbindEventfromOnMediaOpenFailed _mediaplayer_2 (EventDispatchers|CreateEvent :!event On%20Media%20Open%20Failed))
      (Media|MediaPlayer|UnbindEventfromOnMediaClosed _mediaplayer_2 (EventDispatchers|CreateEvent :!event On%20Media%20Closed))
      (Media|MediaPlayer|UnbindEventfromOnEndReached _mediaplayer_2 (EventDispatchers|CreateEvent :!event On%20Media%20End))
      (Media|MediaPlayer|UnbindEventfromOnMediaOpened _mediaplayer_2 (EventDispatchers|CreateEvent :!event On%20Media%20Opened)))))

(fn BindMediaSourceEvents
  (params (Bind :type Boolean))
  (bind _mediasource (Variables|MediaPlayer|GetMediaSource))
  (bind _mediasource_2 (Variables|MediaPlayer|GetMediaSource))
  (if Bind
    (Stream|BindEventtoOnVideoUrlSuccess _mediasource (EventDispatchers|CreateEvent :!event On%20Video%20Url%20Success))
    (Stream|BindEventtoOnVideoUrlFailed _mediasource (EventDispatchers|CreateEvent :!event On%20Video%20Url%20Failed))
    (else
      (Stream|UnbindEventfromOnVideoUrlFailed _mediasource_2 (EventDispatchers|CreateEvent :!event On%20Video%20Url%20Failed))
      (Stream|UnbindEventfromOnVideoUrlSuccess _mediasource_2 (EventDispatchers|CreateEvent :!event On%20Video%20Url%20Success)))))

(fn CheckBlacklist
  (if (Utilities|String|EqualExactly(String) (Variables|Default|GetScreenName) "RiskyReels")
    (return (or (not (GameplayTags|HasAnyTags (Class|FortGameStateAthena|GetAthenaPlaylistContextTags (Fortnite|Athena|GameState|GetGameStateAthena)) (FortGlobals|GetElectraStreamPlaylistDenylistFilter))) (Replay|GetIsPlayingReplay)))
    (else
      (return true))))

(fn CheckBlocks
  (return (Utilities|Enum|Equal(Enum) (Player|ShouldProtectPlayerfromContent (Variables|MediaPlayer|GetMediaSource)) "UCPAudio")))

(fn CheckLOS
  (params (SkipUserCheck :type Boolean) (ForceTrue :type Boolean) (SkipSet :type Boolean))
  (bind _returnvalue (Fortnite|GetFirstLocalFortPlayerController true))
  (bind _returnvalue_2 (Pawn|GetControlledPawn _returnvalue))
  (bind _returnvalue_3 (Pawn|GetBaseAimRotation _returnvalue_2))
  (bind _returnvalue_4 (Math|Vector|GetForwardVector _returnvalue_3))
  (bind _screenactor (Variables|Default|GetScreenActor))
  (bind _returnvalue_5 (TargetSelection|GetDirectionActortoActor _returnvalue_2 _screenactor))
  (bind _returnvalue_6 (Math|Vector|GetForwardVector _returnvalue_5))
  (bind _returnvalue_7 (Math|Vector|DotProduct _returnvalue_6 _returnvalue_4))
  (bind _returnvalue_8 (Fortnite|GetFirstLocalFortPlayerController true))
  (bind _returnvalue_9 (Pawn|GetControlledPawn _returnvalue_8))
  (bind _returnvalue_10 (> _returnvalue_7 0.6))
  (LocalVariables|SetSkipUserCheckInt SkipUserCheck)
  (LocalVariables|SetSkipSetInternal SkipSet)
  (if (not ForceTrue)
    (if (Utilities|IsValid _returnvalue_9 :!class /Script/Engine.KismetSystemLibrary)
      (if (Utilities|IsValid (Variables|Default|GetScreenActor) :!class /Script/Engine.KismetSystemLibrary)
        (Utilities|IsinVehicle _returnvalue_9
          (:true
            (bind _output_get (Variables|Default|SetScreenLOS true))
            (return _output_get))
          (:false
            (bind _returnvalue_11 (Fortnite|CheckLineOfSightToActorWithChannel (Math|Vector|MakeVector (.x (Transformation|GetActorLocation _returnvalue_2)) (.y (Transformation|GetActorLocation _returnvalue_2)) (+ (.z (Transformation|GetActorLocation _returnvalue_2)) (Variables|/Script/Engine.Pawn|GetBaseEyeHeight _returnvalue_2))) _screenactor "ECC_Visibility" _returnvalue_2))
            (if (not (LocalVariables|GetSkipUserCheckInt))
              (bind _enabled (PiP|UserPIPEnabled))
              (if _enabled
                _
                (elif (LocalVariables|GetSkipSetInternal)
                  (return true)
                  (else
                    (bind _output_get_2 (Variables|Default|SetScreenLOS true))
                    (return _output_get_2)))))
            (if (> (Math|Vector|DotProduct (Math|Vector|GetForwardVector (TargetSelection|GetDirectionActortoActor _returnvalue_2 _screenactor)) (Transformation|GetActorRightVector _screenactor)) 0.0)
              (if (LocalVariables|GetSkipSetInternal)
                (return false)
                (else
                  (bind _output_get_3 (Variables|Default|SetScreenLOS false))
                  (return _output_get_3)))
              (elif (Default|ShouldStreambePlaying 2.5e+03)
                (if (LocalVariables|GetSkipSetInternal)
                  (return true)
                  (else
                    (bind _output_get_4 (Variables|Default|SetScreenLOS true))
                    (return _output_get_4)))
                (elif (LocalVariables|GetSkipSetInternal)
                  (return (and _returnvalue_11 _returnvalue_10))
                  (else
                    (bind _output_get_5 (Variables|Default|SetScreenLOS (and _returnvalue_11 _returnvalue_10)))
                    (return _output_get_5)))))))
        (else
          _)))
    (bind _output_get_6 (Variables|Default|SetScreenLOS false))
    (return _output_get_6)
    (else
      (bind _output_get_7 (Variables|Default|SetScreenLOS true))
      (return _output_get_7))))

(fn CheckPiPDisabled
  (return (or (Variables|PiP|GetPiPForceOff) (and (not (Mesh|GetDataBool (Variables|MeshVars|GetMesh_PiPMirrorEnable))) (not (Mesh|GetDataBool (Variables|MeshVars|GetMesh_PiPEnable)))))))

(fn CheckScreen
  (params (ScreenName :type String))
  (return (or (Utilities|String|EqualExactly(String) ScreenName "all") (Utilities|String|EqualExactly(String) ScreenName (Variables|Default|GetScreenName)))))

(fn CheckVideoDisabled
  (params (StandAlonePIP :type Boolean))
  (bind _mediasource (Utilities|GetPIPMediaSource StandAlonePIP)
    (:New
      (return (or (Utilities|Enum|Equal(Enum) (Player|ShouldProtectPlayerfromContent _mediasource) "UCPAudio") (Utilities|Enum|Equal(Enum) (Player|ShouldProtectPlayerfromContent _mediasource) "UCPAudio"))))
    (:Old
      (bind _returnvalue (Class|LiveStreamStandalonePlayer|GetPIPStandaloneError))
      (return (Math|Boolean|ORBoolean (select StandAlonePIP _returnvalue false) (Utilities|Enum|Equal(Enum) (Player|ShouldProtectPlayerfromContent _mediasource) "UCPAudio") (Utilities|Enum|Equal(Enum) (Player|ShouldProtectPlayerfromContent _mediasource) "UCPAudio"))))))

(fn CheckVideoDisabledRaw
  (params (StandalonePIP :type Boolean))
  (bind _mediasource (Utilities|GetPIPMediaSource StandalonePIP)
    (:New
      _)
    (:Old
      _))
  (return (Utilities|String|Contains (Variables|/Script/FortniteGame.FortStreamMediaSource|GetProtectUserFromAVSettings (select StandalonePIP _mediasource (Variables|MediaPlayer|GetMediaSource))) "v")))

(fn ControlNewMirrorPIP
  (params (Enable :type Boolean))
  (bind _medianame (LocalVariables|GetMediaName))
  (if Enable
    (bind _mediatexture (Utilities|GetPIPTexture)
      (:Out
        (Class|FortMediaEventsSubsystem|BroadcastStreamAssets (EngineSubsystems|GetFortMediaEventsSubsystem) (Utilities|Struct|MakeFortMediaEventsStreamAssets _medianame false _mediatexture 0 "/Game/Movies/Season5/Blueprints/Screens_AudioBus.Screens_AudioBus" "/Game/Sounds/SubMixes/Fortnite_SFX.Fortnite_SFX" "/Game/Sounds/SubMixes/Fortnite_LicensedContent_3D.Fortnite_LicensedContent_3D" -1.0 (Variables|MediaPlayer|GetMediaSource) (Variables|MediaPlayer|GetMediaPlayer)))
        (bind _output_get (Variables|PiP|SetPiPEnabled true))
        (CallFunction|EnablePiPDistanceCheck _output_get)))
    (else
      (Class|FortMediaEventsSubsystem|ClearStreamAssets (EngineSubsystems|GetFortMediaEventsSubsystem) _medianame)
      (CallFunction|EnablePiPDistanceCheck false)
      (Variables|PiP|SetPiPEnabled false))))

(fn ControlNewSAPIP
  (params (Enable :type Boolean))
  (bind _returnvalue (EngineSubsystems|GetFortMediaEventsSubsystem))
  (Variables|PiP|SetPiPEnabled Enable)
  (Variables|PiP|SetPiPStandAlone Enable)
  (if Enable
    (EventDispatchers|BindEventtoMediaSourceReturn _returnvalue (EventDispatchers|CreateEvent :!event ControlNewSAPIPCont))
    (Class|FortMediaEventsSubsystem|RequestExternalMediaControllerMediaSource _returnvalue)
    (EventDispatchers|BindEventtoOnMediaSourceSelected _returnvalue (EventDispatchers|CreateEvent :!event UpdateSAMS))
    (else
      (CallFunction|EnablePiPDistanceCheck false)
      (EventDispatchers|UnbindEventfromOnMediaSourceSelected _returnvalue (EventDispatchers|CreateEvent :!event UpdateSAMS))
      (Class|FortMediaEventsSubsystem|ResumeExternalMediaControllers _returnvalue))))

(fn ControlNewSAPIPCont
  (params (MediaSource :type Fort%20Stream%20Media%20Source%20Object%20Reference))
  (Variables|PiP|SetNewSAPIPMS MediaSource)
  (EventDispatchers|UnbindEventfromMediaSourceReturn (EngineSubsystems|GetFortMediaEventsSubsystem) (EventDispatchers|CreateEvent :!event ControlNewSAPIPCont))
  (CallFunction|EnablePiPDistanceCheck true))

(fn DebounceLOS
  (bind _returnvalue (Fortnite|Sync|Time|GetServerWorldTimeSeconds))
  (bind _lastlosswitch (Variables|PiP|GetLastLOSSwitch))
  (if (== _lastlosswitch 0.0)
    (Variables|PiP|SetLastLOSSwitch _returnvalue)
    (elif (> (- _returnvalue _lastlosswitch) (Variables|PiP|GetLOSSwitchDebounceTime))
      (Variables|PiP|SetLastLOSSwitch 0.0)
      (else
        (return true))))
  (return false))

(fn DrawPIPBlackout
  (params (StandalonePIP :type Boolean))
  (bind _blocked (ContentProtection|CheckVideoDisabledRaw StandalonePIP))
  (Class|LiveStreamHUDElement|SetPIPEOSGSBlackout _blocked))

(fn EnablePIPAudio
  (Class|LiveStreamHUDElement|SetMirrorPIPAudioDisable (or (Utilities|Enum|Equal(Enum) (Player|ShouldProtectPlayerfromContent (Variables|MediaPlayer|GetMediaSource)) "UCPAudio") (Utilities|Enum|Equal(Enum) (Player|ShouldProtectPlayerfromContent (Variables|MediaPlayer|GetMediaSource)) "UCPAudio"))))

(fn EnableShareBlocking
  (params (Enable :type Boolean))
  (bind _mediasource (Variables|MediaPlayer|GetMediaSource))
  (bind _bsharelock (Variables|/Script/FortniteGame.FortStreamMediaSource|GetbShareLock _mediasource))
  (if Enable
    (if (and _bsharelock (not (Variables|ShareBlocking|GetDisabledSharing)))
      (Player|DisableSharing _mediasource)
      (Variables|ShareBlocking|SetDisabledSharing true))
    (bind _returnvalue (Utilities|Time|SetTimerbyEvent (EventDispatchers|CreateEvent :!event SharingCheck) 0.05 true))
    (Variables|ShareBlocking|SetSharingCheckTimer _returnvalue)
    (if (Utilities|String|Contains (Variables|/Script/FortniteGame.FortStreamMediaSource|GetProtectUserFromAVSettings (Variables|MediaPlayer|GetMediaSource)) "a")
      (bind _returnvalue_2 (Utilities|Time|SetTimerbyEvent (EventDispatchers|CreateEvent :!event EnablePIPAudio) 0.05 true))
      (Variables|Default|SetUCPMirrorPIPTimer _returnvalue_2))
    (else
      (Utilities|Time|ClearandInvalidateTimerbyHandle (Variables|ShareBlocking|GetSharingCheckTimer))
      (Utilities|Time|ClearandInvalidateTimerbyHandle (Variables|Default|GetUCPMirrorPIPTimer))
      (if (and _bsharelock (Variables|ShareBlocking|GetDisabledSharing))
        (Player|ReinstateSharing _mediasource)
        (Variables|ShareBlocking|SetDisabledSharing false)))))

(fn End
  (Screen|StopScreenTimers)
  (Variables|Default|SetIsVideoTestEnabled false)
  (CallFunction|CheckDistancetoScreen)
  (ContentProtection|EnableShareBlocking false)
  (Class|FortMediaUCPInView|FireUCPInViewCallback false)
  (Variables|Default|SetEnteredStreamingZone false))

(fn EndPlaybackClosePIP
  (bind _screenname (Variables|Default|GetScreenName))
  (Variables|PiP|SetPiPForceOff true)
  (if (Variables|PiP|GetPiPEnabled)
    (if (Variables|PiP|GetPiPStandAlone)
      (CallFunction|StartPiPStandalone _screenname "" false "-")
      (else
        (CallFunction|SetPiPMirror _screenname false "-")))
    (return)))

(fn FlipAudioSubmix
  (params (Default :type Boolean))
  (bind _screenmediasound (Variables|Default|GetScreenMediaSound))
  (if Default
    (Audio|Components|Audio|SetSubmixSend _screenmediasound "/Game/Sounds/SubMixes/Fortnite_LicensedContent_3D.Fortnite_LicensedContent_3D" 0.0 :!class /Script/AudioMixer.SynthComponent)
    (Audio|Components|Audio|SetSubmixSend _screenmediasound "/Game/Sounds/SubMixes/Fortnite_SFX.Fortnite_SFX" 1.0 :!class /Script/AudioMixer.SynthComponent)
    (else
      (Audio|Components|Audio|SetSubmixSend _screenmediasound "/Game/Sounds/SubMixes/Fortnite_SFX.Fortnite_SFX" 0.0 :!class /Script/AudioMixer.SynthComponent)
      (Audio|Components|Audio|SetSubmixSend _screenmediasound "/Game/Sounds/SubMixes/Fortnite_LicensedContent_3D.Fortnite_LicensedContent_3D" 1.0 :!class /Script/AudioMixer.SynthComponent))))

(fn GetData_Bool
  (params (In :type Mesh%20Region%20Lock%20Data%20Bool%20Structure))
  (return (Utilities|Struct|BreakMeshRegionLockDataBool In)))

(fn GetData_DateTime
  (params (In :type Mesh%20Region%20Lock%20Data%20Date%20Time%20Structure))
  (return (Utilities|Struct|BreakMeshRegionLockDataDateTime In)))

(fn GetData_Float
  (params (In :type Mesh%20Region%20Lock%20Data%20Float%20Structure))
  (return (Utilities|Struct|BreakMeshRegionLockDataFloat In)))

(fn GetData_Int
  (params (In :type Mesh%20Region%20Lock%20Data%20Int%20Structure))
  (return (Utilities|Struct|BreakMeshRegionLockDataInt In)))

(fn GetData_Int64
  (params (In :type Mesh%20Region%20Lock%20Data%20Int%2064%20Structure))
  (return (Utilities|Struct|BreakMeshRegionLockDataInt64 In)))

(fn GetData_String
  (params (In :type Mesh%20Region%20Lock%20Data%20String%20Structure))
  (return (Utilities|Struct|BreakMeshRegionLockDataString In)))

(fn GetMeshData_Bool
  (params (Array :type Array%20of%20Mesh%20Region%20Lock%20Data%20Bool%20Structures) (Default :type Boolean))
  (LocalVariables|SetFound false)
  (LocalVariables|SetValue (Utilities|Struct|MakeMeshRegionLockDataBool Default))
  (if (!= (Utilities|Array|Length Array) 0)
    (bind (_arrayindex _arrayelement) (Utilities|Array|ReverseforEachLoop Array)
      (:LoopBody
        (bind _allowed (CallFunction|RegionIsAllowed (.RegionData (Utilities|Struct|BreakMeshRegionLockDataBool _arrayelement))))
        (if (and _allowed (not (LocalVariables|GetFound)))
          (LocalVariables|SetFound true)
          (LocalVariables|SetValue (Utilities|Struct|MakeMeshRegionLockDataBool (Utilities|Struct|BreakMeshRegionLockDataBool _arrayelement) (.RegionData (Utilities|Struct|BreakMeshRegionLockDataBool _arrayelement))))))
      (:Completed
        _)))
  (return (LocalVariables|GetValue)))

(fn GetMeshData_DateTime
  (params (Array :type Array%20of%20Mesh%20Region%20Lock%20Data%20Date%20Time%20Structures) (Default :type Date%20Time%20Structure))
  (LocalVariables|SetFound false)
  (LocalVariables|SetValue (Utilities|Struct|MakeMeshRegionLockDataDateTime Default))
  (if (!= (Utilities|Array|Length Array) 0)
    (bind (_arrayindex _arrayelement) (Utilities|Array|ReverseforEachLoop Array)
      (:LoopBody
        (bind _allowed (CallFunction|RegionIsAllowed (.RegionData (Utilities|Struct|BreakMeshRegionLockDataDateTime _arrayelement))))
        (if (and _allowed (not (LocalVariables|GetFound)))
          (LocalVariables|SetFound true)
          (LocalVariables|SetValue (Utilities|Struct|MakeMeshRegionLockDataDateTime (Utilities|Struct|BreakMeshRegionLockDataDateTime _arrayelement) (.RegionData (Utilities|Struct|BreakMeshRegionLockDataDateTime _arrayelement))))))
      (:Completed
        _)))
  (return (LocalVariables|GetValue)))

(fn GetMeshData_Float
  (params (Array :type Array%20of%20Mesh%20Region%20Lock%20Data%20Float%20Structures) (Default :type Float%20%28double-precision%29))
  (LocalVariables|SetFound false)
  (LocalVariables|SetValue (Utilities|Struct|MakeMeshRegionLockDataFloat Default))
  (if (!= (Utilities|Array|Length Array) 0)
    (bind (_arrayindex _arrayelement) (Utilities|Array|ReverseforEachLoop Array)
      (:LoopBody
        (bind _allowed (CallFunction|RegionIsAllowed (.RegionData (Utilities|Struct|BreakMeshRegionLockDataFloat _arrayelement))))
        (if (and _allowed (not (LocalVariables|GetFound)))
          (LocalVariables|SetFound true)
          (LocalVariables|SetValue (Utilities|Struct|MakeMeshRegionLockDataFloat (Utilities|Struct|BreakMeshRegionLockDataFloat _arrayelement) (.RegionData (Utilities|Struct|BreakMeshRegionLockDataFloat _arrayelement))))))
      (:Completed
        _)))
  (return (LocalVariables|GetValue)))

(fn GetMeshData_Int
  (params (Array :type Array%20of%20Mesh%20Region%20Lock%20Data%20Int%20Structures) (Default :type Integer))
  (LocalVariables|SetFound false)
  (LocalVariables|SetValue (Utilities|Struct|MakeMeshRegionLockDataInt Default))
  (if (!= (Utilities|Array|Length Array) 0)
    (bind (_arrayindex _arrayelement) (Utilities|Array|ReverseforEachLoop Array)
      (:LoopBody
        (bind _allowed (CallFunction|RegionIsAllowed (.RegionData (Utilities|Struct|BreakMeshRegionLockDataInt _arrayelement))))
        (if (and _allowed (not (LocalVariables|GetFound)))
          (LocalVariables|SetFound true)
          (LocalVariables|SetValue (Utilities|Struct|MakeMeshRegionLockDataInt (Utilities|Struct|BreakMeshRegionLockDataInt _arrayelement) (.RegionData (Utilities|Struct|BreakMeshRegionLockDataInt _arrayelement))))))
      (:Completed
        _)))
  (return (LocalVariables|GetValue)))

(fn GetMeshData_Int64
  (params (Array :type Array%20of%20Mesh%20Region%20Lock%20Data%20Int%2064%20Structures) (Default :type Integer64))
  (LocalVariables|SetFound false)
  (LocalVariables|SetValue (Utilities|Struct|MakeMeshRegionLockDataInt64 Default))
  (if (!= (Utilities|Array|Length Array) 0)
    (bind (_arrayindex _arrayelement) (Utilities|Array|ReverseforEachLoop Array)
      (:LoopBody
        (bind _allowed (CallFunction|RegionIsAllowed (.RegionData (Utilities|Struct|BreakMeshRegionLockDataInt64 _arrayelement))))
        (if (and _allowed (not (LocalVariables|GetFound)))
          (LocalVariables|SetFound true)
          (LocalVariables|SetValue (Utilities|Struct|MakeMeshRegionLockDataInt64 (Utilities|Struct|BreakMeshRegionLockDataInt64 _arrayelement) (.RegionData (Utilities|Struct|BreakMeshRegionLockDataInt64 _arrayelement))))))
      (:Completed
        _)))
  (return (LocalVariables|GetValue)))

(fn GetMeshData_String
  (params (Array :type Array%20of%20Mesh%20Region%20Lock%20Data%20String%20Structures) (Default :type String))
  (LocalVariables|SetFound false)
  (LocalVariables|SetValue (Utilities|Struct|MakeMeshRegionLockDataString Default))
  (if (!= (Utilities|Array|Length Array) 0)
    (bind (_arrayindex _arrayelement) (Utilities|Array|ReverseforEachLoop Array)
      (:LoopBody
        (bind _allowed (CallFunction|RegionIsAllowed (.RegionData (Utilities|Struct|BreakMeshRegionLockDataString _arrayelement))))
        (if (and _allowed (not (LocalVariables|GetFound)))
          (LocalVariables|SetFound true)
          (LocalVariables|SetValue (Utilities|Struct|MakeMeshRegionLockDataString (Utilities|Struct|BreakMeshRegionLockDataString _arrayelement) (.RegionData (Utilities|Struct|BreakMeshRegionLockDataString _arrayelement))))))
      (:Completed
        _)))
  (return (LocalVariables|GetValue)))

(fn GetRegionInfo
  (params (SetupMesh :type Boolean :default true))
  (bind _fortmediaregionlock (Variables|Default|GetFortMediaRegionLock))
  (Variables|Default|SetSetupMeshFromRegionInfo SetupMesh)
  (if (not (Variables|/Script/FortniteGame.FortRuntimeOptions|GetbDisableGRL (Options|GetRuntimeOptions)))
    (bind _ispopulated (CallFunction|IsRegionInfoPopulated))
    (if (not _ispopulated)
      (EventDispatchers|BindEventtoOnMediaRegionLockLocaleComplete _fortmediaregionlock (EventDispatchers|CreateEvent :!event GetRegionInfoComplete))
      (Class|FortMediaRegionLockComponent|SetEventCallback _fortmediaregionlock true)
      (Class|FortMediaRegionLockComponent|GetLocale _fortmediaregionlock (Fortnite|GetFirstLocalFortPlayerController true))
      (else
        _)))
  (if (Variables|Default|GetSetupMeshFromRegionInfo)
    (CallFunction|SetupMeshTrigger)))

(fn GetRegionInfoComplete
  (params (Continent :type String) (Country :type String) (Subdivisions :type Array%20of%20Strings))
  (Variables|MeshVars|SetContinent Continent)
  (Variables|MeshVars|SetCountry Country)
  (Variables|MeshVars|SetSubdivisions Subdivisions)
  (if (Variables|Default|GetSetupMeshFromRegionInfo)
    (CallFunction|SetupMeshTrigger)))

(fn HandleMediaCtrl
  (params (Create :type Boolean))
  (bind _mediaplayerctrl (Variables|Default|GetMediaPlayerCtrl))
  (if Create
    (bind _returnvalue (Game|ConstructObjectfromClass "/Script/FortniteGame.FortMediaPlayerCtrl" self))
    (Variables|Default|SetMediaPlayerCtrl _returnvalue)
    (Variables|/Script/FortniteGame.FortMediaPlayerCtrl|SetMediaPlayer (Variables|MediaPlayer|GetMediaPlayer) _mediaplayerctrl)
    (Variables|/Script/FortniteGame.FortMediaPlayerCtrl|SetMediaSource (Variables|MediaPlayer|GetMediaSource) _mediaplayerctrl)
    (else
      (bind _mediaplayerctrl_2 (Variables|Default|GetMediaPlayerCtrl)
        (:then
          (Variables|/Script/FortniteGame.FortMediaPlayerCtrl|SetMediaPlayer 0 _mediaplayerctrl_2)
          (Variables|/Script/FortniteGame.FortMediaPlayerCtrl|SetMediaSource 0 _mediaplayerctrl_2)
          (Variables|Default|SetMediaPlayerCtrl 0))
        (:else)))))

(fn IsRegionInfoPopulated
  (return (not (or (Utilities|String|IsEmpty (Variables|MeshVars|GetCountry)) (Utilities|String|IsEmpty (Variables|MeshVars|GetContinent))))))

(fn RegionIsAllowed
  (params (RegionData :type String))
  (bind _country (Variables|MeshVars|GetCountry))
  (bind _returnvalue (Utilities|Array|FindItem (Utilities|String|ParseIntoArray (Utilities|Array|Get(aref) (LocalVariables|GetRegionDataArray) 2) ",") _country))
  (bind _continent (Variables|MeshVars|GetContinent))
  (bind _returnvalue_2 (Utilities|Array|FindItem (Utilities|String|ParseIntoArray (Utilities|Array|Get(aref) (LocalVariables|GetRegionDataArray) 2) ",") _continent))
  (bind _regiondataarray (LocalVariables|GetRegionDataArray))
  (bind _output (Utilities|Array|Get(aref) _regiondataarray 2))
  (bind _returnvalue_3 (Utilities|String|ParseIntoArray _output ","))
  (bind _regiondataarray_2 (LocalVariables|GetRegionDataArray))
  (bind _output_2 (Utilities|Array|Get(aref) _regiondataarray_2 3))
  (bind _returnvalue_4 (Utilities|String|ParseIntoArray _output_2 ","))
  (if (Utilities|String|EqualExactly(String) RegionData (Variables|MeshVars|GetAllRegions))
    (return true)
    (else
      (LocalVariables|SetRegionDataArray (Utilities|String|ParseIntoArray RegionData ":"))
      (if (Utilities|String|IsEmpty RegionData)
        (return false)
        (else
          (LocalVariables|SetListType (Utilities|Array|Get(aref) _regiondataarray 1))
          (switch string (Utilities|Array|Get(aref) _regiondataarray)
            (:country
              (if (== (Utilities|Array|Length _regiondataarray_2) 4)
                (LocalVariables|SetAllowedInternal false)
                (bind _array_element (Utilities|Array|ForEachLoopwithBreak (Variables|MeshVars|GetSubdivisions) :!label _label)
                  (:LoopBody
                    (switch string (LocalVariables|GetListType)
                      (:whitelist
                        (if (!= (Utilities|Array|FindItem _returnvalue_4 _array_element) -1)
                          _))
                      (:blacklist
                        (if (== (Utilities|Array|FindItem _returnvalue_4 _array_element) -1)
                          _)))
                    (LocalVariables|SetAllowedInternal true)
                    (goto _label Break))
                  (:Completed
                    (return (LocalVariables|GetAllowedInternal))))
                (else
                  (switch string (LocalVariables|GetListType)
                    (:whitelist
                      (if (not (!= _returnvalue -1))
                        (return false)))
                    (:blacklist
                      (if (not (== _returnvalue -1))
                        (return false)))))))
            (:continent
              (switch string (LocalVariables|GetListType)
                (:whitelist
                  (if (not (!= _returnvalue_2 -1))
                    (return false)))
                (:blacklist
                  (if (not (== _returnvalue_2 -1))
                    (return false))))))
          (return true))))))

(fn RenderPIP
  (params (Enable :type Boolean))
  (Utilities|PIPVersion
    (:New
      (Class|FortMediaEventsSubsystem|SetExternalMediaControllerEnabled (EngineSubsystems|GetFortMediaEventsSubsystem) Enable))
    (:Old
      (Variables|/Script/FortniteGame.FortRuntimeOptions|SetbEnableLiveStreamInMatch Enable (Options|GetRuntimeOptions)))))

(fn SendAnalytics
  (for _array_element (Fortnite|GetLocalFortPlayerControllers true)
    (Movie|SendStreamingAnalytics _array_element (Variables|MediaPlayer|GetMediaPlayer))
    (return)))

(fn SendRelevancyAnalytics
  (for _array_element (Fortnite|GetLocalFortPlayerControllers true)
    (Analytics|FireEventStreamingMediaRelevancy _array_element (Math|Timespan|GetTotalSeconds (Math|DateTime|DateTime-DateTime (Math|DateTime|UTCNow) (Variables|Default|GetTimeEnteredStreamingZone))))))

(fn SetAudioSubmix
  (bind _screenmediasound (Variables|Default|GetScreenMediaSound))
  (Class|LiveStreamHUDElement|SetMirrorPIPAudioSubmix (not (Utilities|String|Contains (Variables|/Script/FortniteGame.FortStreamMediaSource|GetProtectUserFromAVSettings (Variables|MediaPlayer|GetMediaSource)) "a")) true)
  (if (Variables|Default|GetUseScreenMSC)
    (Audio|Components|Audio|SetSubmixSend _screenmediasound "/Game/Sounds/SubMixes/Fortnite_LicensedContent_3D.Fortnite_LicensedContent_3D" 0.0 :!class /Script/AudioMixer.SynthComponent)
    (Audio|Components|Audio|SetSubmixSend _screenmediasound "/Game/Sounds/SubMixes/Fortnite_SFX.Fortnite_SFX" 1.0 :!class /Script/AudioMixer.SynthComponent)
    (else
      (Audio|Components|Audio|SetSubmixSend _screenmediasound "/Game/Sounds/SubMixes/Fortnite_SFX.Fortnite_SFX" 0.0 :!class /Script/AudioMixer.SynthComponent)
      (Audio|Components|Audio|SetSubmixSend _screenmediasound "/Game/Sounds/SubMixes/Fortnite_LicensedContent_3D.Fortnite_LicensedContent_3D" 1.0 :!class /Script/AudioMixer.SynthComponent))))

(fn SetPIPAnalyticsData
  (params (Reset :type Boolean))
  (bind (_days _hours _minutes _seconds _milliseconds) (Math|Timespan|BreakTimespan (Media|MediaPlayer|GetTime (Variables|MediaPlayer|GetMediaPlayer))))
  (if Reset
    (Class|LiveStreamHUDElement|SetMirroredPIPAnalyicsMetadata (Variables|Default|GetResetData) 0 (Variables|Default|GetResetData))
    (else
      (Class|LiveStreamHUDElement|SetMirroredPIPAnalyicsMetadata (Utilities|String|StringToName (Variables|MediaPlayer|GetVideoUID)) (Math|Integer64|integer64+integer64 (* (Math|Conversions|ToInteger64(Integer) _days) 86400000) (* (Math|Conversions|ToInteger64(Integer) _hours) 3600000) (* (Math|Conversions|ToInteger64(Integer) _minutes) 60000) (* (Math|Conversions|ToInteger64(Integer) _seconds) 1000) (Math|Conversions|ToInteger64(Integer) _milliseconds)) (Media|MediaPlayer|GetPlayerName (Variables|MediaPlayer|GetMediaPlayer))))))

(fn SetStartPlayTime
  (bind _string (CallFunction|UtctoString (Mesh|GetDataDateTime (Variables|MeshVars|GetMesh_RootUTCTIme))))
  (bind _string_2 (CallFunction|UtctoString (Variables|Default|GetLocalPlayUTCTime)))
  (Fortnite|FortShippingLog 0 (Utilities|String|Append "StreamTrigger:  Set start play time RootUtc: " _string "
PlayUTC: " _string_2))
  (if (Variables|MediaPlayer|GetForceStartPlayTime)
    (Variables|MediaPlayer|SetForceStartPlayTime false)
    (else
      (Fortnite|FortShippingLog 0 (Utilities|String|Append "StreamTrigger: Set start play time 
" "ServerWorldTimeSeconds = " (Utilities|String|ToString(Float) (Fortnite|Sync|Time|GetServerWorldTimeSeconds)) "
Replicated PlayServerTimeSeconds: " (Utilities|String|ToString(Float) (Variables|Default|GetLocalPlayServerTimeSeconds))))
      (bind _output_get (Variables|Default|SetStartPlayTime (Math|Timespan|GetTotalSeconds (Math|DateTime|DateTime-DateTime (Mesh|GetDataDateTime (Variables|MeshVars|GetMesh_RootUTCTIme)) (Variables|Default|GetLocalPlayUTCTime)))))
      (Fortnite|FortShippingLog 0 (Utilities|String|Append "StartPlayTime= " (Utilities|String|ToString(Float) _output_get))))))

(fn SharingCheck
  (Default|CallVideoSharingCheck))

(fn ShouldStreamBePlaying
  (params (Range :type Float%20%28double-precision%29))
  (Utilities|IsValid (Variables|Default|GetScreenMesh)
    (:Is%20Valid
      (return (Stream|ShouldStreambePlaying (Variables|MediaPlayer|GetMediaSource) (Variables|Default|GetScreenMesh) Range)))
    (:Is%20Not%20Valid
      (return (Stream|ShouldStreambePlaying (Variables|MediaPlayer|GetMediaSource) (Variables|Default|GetStaticMesh) Range)))))

(fn StopScreenTimers
  (Utilities|Time|ClearandInvalidateTimerbyHandle (Variables|Default|GetCheckDistanceTimer))
  (Utilities|Time|ClearandInvalidateTimerbyHandle (Variables|Default|GetBandwidthCheckLimitTimer))
  (Utilities|Time|ClearandInvalidateTimerbyHandle (Variables|Default|GetStartPlayingTimer))
  (Utilities|Time|ClearandInvalidateTimerbyHandle (Variables|Default|GetCheckUCPFromStartTimer))
  (ContentProtection|UCPInViewTimer false))

(fn StreamFailCheck
  (bind _mediaplayer (Variables|MediaPlayer|GetMediaPlayer))
  (bind _returnvalue (Media|MediaPlayer|HasError _mediaplayer))
  (bind _mediasource (Variables|MediaPlayer|GetMediaSource))
  (bind _returnvalue_2 (Stream|HasLocalFilePlayback _mediasource))
  (bind _videoretrycounter (Variables|Default|GetVideoRetryCounter))
  (if (and _returnvalue _returnvalue_2)
    (Variables|Default|SetFallbackMediaSource (Variables|/Script/FortniteGame.FortStreamMediaSource|GetLocalFilePlaybackAsset _mediasource))
    (Variables|MediaPlayer|SetVideoEnableLocalFallback true)
    (elif (Math|Boolean|ANDBoolean _returnvalue (not _returnvalue_2) (< _videoretrycounter (Variables|Default|GetVideoMaxRetries)))
      (Math|Integer|IncrementInt _videoretrycounter
        (:%20%20
          (Variables|MediaPlayer|SetStreamingFailed true)))
      (else
        (Variables|MediaPlayer|SetStreamingFailed false)
        (Variables|Default|SetVideoRetryCounter 0)))))

(fn TelevisionOff
  (Variables|MediaPlayer|SetMediaStarted_PIPLOS false)
  (Variables|MediaPlayer|SetDelayVideoStart false)
  (Class|LiveStreamStandalonePlayer|UnblockPlayer "StreamInWorldActive")
  (CallFunction|BindPlayerEvents false)
  (CallFunction|SendAnalytics)
  (Utilities|Time|ClearandInvalidateTimerbyHandle (Variables|Default|GetStartPlayingTimer))
  (Default|CallVideoStoppedEventDispatcher)
  (if (Variables|Default|GetTempAudioUsed)
    (Audio|ActivateTempMSC false))
  (Audio|ActivateScreenMSC false))

(fn UCPInViewCheck
  (if (Utilities|String|Contains (Variables|/Script/FortniteGame.FortStreamMediaSource|GetProtectUserFromAVSettings (Variables|MediaPlayer|GetMediaSource)) "v")
    (bind _los (Screen|CheckLOS true false true))
    (Class|FortMediaUCPInView|FireUCPInViewCallback _los)))

(fn UCPInViewTimer
  (params (Enable :type Boolean))
  (if (Utilities|String|Contains (Variables|/Script/FortniteGame.FortStreamMediaSource|GetProtectUserFromAVSettings (Variables|MediaPlayer|GetMediaSource)) "v")
    (if Enable
      (bind _returnvalue (Utilities|Time|SetTimerbyEvent (EventDispatchers|CreateEvent :!event UCPInViewCheck) 0.05 true))
      (Variables|Default|SetUCPCheckTimerHandle _returnvalue)
      (else
        (Utilities|Time|ClearandInvalidateTimerbyHandle (Variables|Default|GetUCPCheckTimerHandle))))))

(fn UpdateMeshVariables
  (bind _meshtrigger (Variables|MeshVars|GetMeshTrigger))
  (Utilities|IsValid (Variables|MeshVars|GetMeshTrigger)
    (:Is%20Valid
      (Utilities|FlowControl|Sequence
        (:then_0
          (bind _out (Mesh|GetMeshDataInt (Variables|/Game/Movies/Season5/Blueprints/BP_VideoMeshTrigger.BP_VideoMeshTrigger_C|GetRegion_PIPVersion _meshtrigger) -1))
          (Variables|MeshVars|SetMesh_PIPVersion _out)
          (bind _out_2 (Mesh|GetMeshDataBool (Variables|/Game/Movies/Season5/Blueprints/BP_VideoMeshTrigger.BP_VideoMeshTrigger_C|GetRegion_LeaveScreen _meshtrigger)))
          (Variables|MeshVars|SetMesh_LeaveScreen _out_2)
          (Variables|MeshVars|SetMesh_ScreenName (Variables|/Game/Movies/Season5/Blueprints/BP_VideoMeshTrigger.BP_VideoMeshTrigger_C|GetScreenName _meshtrigger))
          (bind _out_3 (Mesh|GetMeshDataString (Variables|/Game/Movies/Season5/Blueprints/BP_VideoMeshTrigger.BP_VideoMeshTrigger_C|GetRegion_VideoUID _meshtrigger) (Variables|MediaPlayer|GetVideoUID_Default)))
          (Variables|MeshVars|SetMesh_VideoUID _out_3)
          (bind _out_4 (Mesh|GetMeshDataDateTime (Variables|/Game/Movies/Season5/Blueprints/BP_VideoMeshTrigger.BP_VideoMeshTrigger_C|GetRegion_PlayUTCTime_GS _meshtrigger)))
          (Variables|MeshVars|SetMesh_PlayUTCTime _out_4)
          (bind _out_5 (Mesh|GetMeshDataDateTime (Variables|/Game/Movies/Season5/Blueprints/BP_VideoMeshTrigger.BP_VideoMeshTrigger_C|GetRegion_RootUtcTime _meshtrigger)))
          (Variables|MeshVars|SetMesh_RootUTCTIme _out_5))
        (:then_1
          (bind _out_6 (Mesh|GetMeshDataBool (Variables|/Game/Movies/Season5/Blueprints/BP_VideoMeshTrigger.BP_VideoMeshTrigger_C|GetRegion_SimpleWarming _meshtrigger)))
          (Variables|MeshVars|SetMesh_SimpleWarming _out_6)
          (bind _out_7 (Mesh|GetMeshDataString (Variables|/Game/Movies/Season5/Blueprints/BP_VideoMeshTrigger.BP_VideoMeshTrigger_C|GetRegion_WarmingCDNWeights _meshtrigger) "\"\""))
          (Variables|MeshVars|SetMesh_WarmingCDNWeights _out_7)
          (bind _out_8 (Mesh|GetMeshDataFloat (Variables|/Game/Movies/Season5/Blueprints/BP_VideoMeshTrigger.BP_VideoMeshTrigger_C|GetRegion_StartWarmingServerTimeSeconds_GS _meshtrigger) -1.0))
          (Variables|MeshVars|SetMesh_StartWarmingServerTimeSeconds _out_8)
          (bind _out_9 (Mesh|GetMeshDataFloat (Variables|/Game/Movies/Season5/Blueprints/BP_VideoMeshTrigger.BP_VideoMeshTrigger_C|GetRegion_VideoPlayStartTime _meshtrigger) -1.0))
          (Variables|MeshVars|SetMesh_VideoPlayStartTime _out_9)
          (bind _out_10 (Mesh|GetMeshDataFloat (Variables|/Game/Movies/Season5/Blueprints/BP_VideoMeshTrigger.BP_VideoMeshTrigger_C|GetRegion_PlayServerTimeSeconds_GS _meshtrigger) -1.0))
          (Variables|MeshVars|SetMesh_PlayServerTimeSeconds _out_10))
        (:then_2
          (bind _out_11 (Mesh|GetMeshDataInt (Variables|/Game/Movies/Season5/Blueprints/BP_VideoMeshTrigger.BP_VideoMeshTrigger_C|GetRegion_MaxBandwidth_GS _meshtrigger) -1))
          (Variables|MeshVars|SetMesh_MaxBandwidth _out_11)
          (bind _out_12 (Mesh|GetMeshDataBool (Variables|/Game/Movies/Season5/Blueprints/BP_VideoMeshTrigger.BP_VideoMeshTrigger_C|GetRegion_PiPMirrorEnable _meshtrigger)))
          (Variables|MeshVars|SetMesh_PiPMirrorEnable _out_12)
          (bind _out_13 (Mesh|GetMeshDataBool (Variables|/Game/Movies/Season5/Blueprints/BP_VideoMeshTrigger.BP_VideoMeshTrigger_C|GetRegion_PiPEnable _meshtrigger)))
          (Variables|MeshVars|SetMesh_PiPEnable _out_13)
          (bind _out_14 (Mesh|GetMeshDataString (Variables|/Game/Movies/Season5/Blueprints/BP_VideoMeshTrigger.BP_VideoMeshTrigger_C|GetRegion_PiPVideoUID _meshtrigger) "\"\""))
          (Variables|MeshVars|SetMesh_PiPVideoUID _out_14)
          (bind _out_15 (Mesh|GetMeshDataBool (Variables|/Game/Movies/Season5/Blueprints/BP_VideoMeshTrigger.BP_VideoMeshTrigger_C|GetRegion_LoopVideo_GS _meshtrigger)))
          (Variables|MeshVars|SetMesh_LoopVideo _out_15))
        (:then_3
          (bind _out_16 (Mesh|GetMeshDataInt64 (Variables|/Game/Movies/Season5/Blueprints/BP_VideoMeshTrigger.BP_VideoMeshTrigger_C|GetRegion_PiPForceCounter _meshtrigger) -1))
          (Variables|MeshVars|SetMesh_PiPForceCounter _out_16)
          (bind _out_17 (Mesh|GetMeshDataBool (Variables|/Game/Movies/Season5/Blueprints/BP_VideoMeshTrigger.BP_VideoMeshTrigger_C|GetRegion_PIPCDNDistroSet _meshtrigger)))
          (Variables|MeshVars|SetMesh_PIPCDNDistroSet _out_17)
          (bind _out_18 (Mesh|GetMeshDataBool (Variables|/Game/Movies/Season5/Blueprints/BP_VideoMeshTrigger.BP_VideoMeshTrigger_C|GetRegion_IsLive _meshtrigger)))
          (Variables|MeshVars|SetMesh_IsPIPLive _out_18)
          (bind _out_19 (Mesh|GetMeshDataFloat (Variables|/Game/Movies/Season5/Blueprints/BP_VideoMeshTrigger.BP_VideoMeshTrigger_C|GetRegion_PiPPlayerServerTimeSeconds_GS _meshtrigger) -1.0))
          (Variables|MeshVars|SetMesh_PiPPlayerServerTimeSeconds _out_19)
          (bind _out_20 (Mesh|GetMeshDataInt (Variables|/Game/Movies/Season5/Blueprints/BP_VideoMeshTrigger.BP_VideoMeshTrigger_C|GetRegion_ScreenDistance _meshtrigger)))
          (if (not (Variables|Default|GetVideoFromCMS))
            (Variables|MeshVars|SetMesh_ScreenDistance _out_20)))
        (:then_4
          (bind _out_21 (Mesh|GetMeshDataString (Variables|/Game/Movies/Season5/Blueprints/BP_VideoMeshTrigger.BP_VideoMeshTrigger_C|GetRegion_CDNWeights _meshtrigger) "\"\""))
          (Variables|MeshVars|SetMesh_CDNWeights _out_21)
          (bind _out_22 (Mesh|GetMeshDataBool (Variables|/Game/Movies/Season5/Blueprints/BP_VideoMeshTrigger.BP_VideoMeshTrigger_C|GetRegion_IsLive _meshtrigger)))
          (Variables|MeshVars|SetMesh_IsLive _out_22))))
    (:Is%20Not%20Valid
      (Development|PrintString (Utilities|String|Append "Mesh Trigger Invalid for screen: " (Variables|Default|GetScreenName))))))

(fn UpdateSAMS
  (params (MediaSource :type Fort%20Stream%20Media%20Source%20Object%20Reference))
  (Variables|PiP|SetNewSAPIPMS MediaSource))

(fn ConstructionScript)

(fn UserPIPEnabled
  (bind _returnvalue (Settings|GetGameUserSettings))
  (bind _asfort_game_user_settings (Utilities|Casting|CastToFortGameUserSettings _returnvalue))
  (return (Utilities|Enum|Equal(Enum) (Variables|/Script/FortniteGame.FortGameUserSettings|GetShowLiveStreamPictureInPictureInMatchV2 _asfort_game_user_settings) "Show")))

(fn UtcToString
  (params (UTC :type Date%20Time%20Structure))
  (bind (_year _month _day _hour _minute _second _millisecond) (Math|DateTime|BreakDateTime UTC))
  (return (Utilities|String|Append (Utilities|String|ToString(Integer) _hour) ":" (Utilities|String|ToString(Integer) _minute) ":" (Utilities|String|ToString(Integer) _second) ":" (Utilities|String|ToString(Integer) _millisecond))))
