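; Wild West campfire prop Blueprint that randomly starts lit or unlit, toggling its light and looping fire sound.
; On BeginPlay, the Remote (non-authority) branch rolls a DropChance and stores the result in bFireActive?; the authority branch does nothing.
; The bFireActive? RepNotify shows or hides the point light and activates or deactivates the fireplace loop cue to match.
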(event EventBeginPlay
  (Networking|SwitchHasAuthority
    (:Remote
      (bind _dropresult (Class|MissionBlueprintFunctionLibrary|DropChance 32))
      (Variables|Default|SetbFireActive? _dropresult))))

(fn OnRep_bFireActive?
  (if (Variables|Default|GetbFireActive?)
    (Rendering|SetVisibility (Variables|Default|GetPointLight) true false)
    (Components|Activation|Activate (Variables|Default|GetFireplace_Loop_Cue) false)
    (else
      (Rendering|SetVisibility (Variables|Default|GetPointLight) false false)
      (Components|Activation|Deactivate (Variables|Default|GetFireplace_Loop_Cue)))))

(fn ConstructionScript)
