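; Blueprint for the Wild West grandfather clock prop in the Athena containers folder.
; It overrides two getters that return the hour hand and minute hand components of the living grandfather clock.
; The construction script is empty, so any clock behavior comes from outside this graph.
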(fn GetHourHand
  (return (Variables|Living_GrandfatherClock|GetHandHour)))

(fn GetMinHand
  (return (Variables|Living_GrandfatherClock|GetHandMinute)))

(fn ConstructionScript)
