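; Wild West wall clock prop that shows time of day and plays ticking audio.
; When updating is enabled, cosmetic-capable instances run UpdateTime immediately and on a looping timer, writing normalized time of day to minute-hand custom primitive data and playing a tick.
; A collision box starts and stops an interior ticking cue for the local player, and death fades out the tick audio.
; SetTickEnabled always clears the update flag before branching on it.
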(event EventBeginPlay
  ;[1 1 1] Clock Tick Update Time
  (Networking|CanExecuteCosmeticEvents
    (:True
      (if (Variables|Default|GetBShouldUpdateTime)
        (Utilities|Time|SetTimerbyFunctionName 0 "UpdateTime" 1.0 true false 1.0 0.5)
        (CallFunction|UpdateTime)))))

(event EventOnDeathPlayEffects
  (params (Damage :type Float%20%28single-precision%29) (DamageTags :type Gameplay%20Tag%20Container%20Structure%20%28by%20ref%29) (Momentum :type Vector :default "0, 0, 0") (HitInfo :type Hit%20Result%20Structure%20%28by%20ref%29) (InstigatedBy :type Fort%20Pawn%20Object%20Reference) (DamageCauser :type Actor%20Object%20Reference) (EffectContext :type Gameplay%20Effect%20Context%20Handle%20Structure))
  (bind _spot_acabin_clockticks_singles_cue (Variables|Default|GetSpot_ACabin_ClockTicks_Singles_Cue))
  ;[1 1 1] On Death Stop Audio Ticking Loop
  (Utilities|IsValid _spot_acabin_clockticks_singles_cue
    (:Is%20Valid
      (if (Utilities|IsValid _spot_acabin_clockticks_singles_cue :!class /Script/Engine.KismetSystemLibrary)
        (Audio|Components|Audio|FadeOut _spot_acabin_clockticks_singles_cue 0.2 0.0 "Linear" :!class /Script/Engine.AudioComponent)))))

(event OnComponentBeginOverlap(Box_Collision)
  (params (OverlappedComponent :type Primitive%20Component%20Object%20Reference) (OtherActor :type Actor%20Object%20Reference) (OtherComp :type Primitive%20Component%20Object%20Reference) (OtherBodyIndex :type Integer) (bFromSweep :type Boolean) (SweepResult :type Hit%20Result%20Structure%20%28by%20ref%29))
  (bind _returnvalue (Fortnite|GetFortPlayerControllerfromActor OtherActor))
  ;[1 1 1] Audio Clock Chimes Timer SFX
  (Utilities|IsValid _returnvalue
    (:Is%20Valid
      (if (Pawn|IsLocalPlayerController _returnvalue)
        (Audio|Components|Audio|Play (Variables|Default|GetSpot_ACabin_Clock_Ticking_Int_Cue) 0.0)))))

(event OnComponentEndOverlap(Box_Collision)
  (params (OverlappedComponent :type Primitive%20Component%20Object%20Reference) (OtherActor :type Actor%20Object%20Reference) (OtherComp :type Primitive%20Component%20Object%20Reference) (OtherBodyIndex :type Integer))
  (bind _returnvalue (Fortnite|GetFortPlayerControllerfromActor OtherActor))
  ;[1 1 1] Audio Clock Chimes Timer SFX
  (Utilities|IsValid _returnvalue
    (:Is%20Valid
      (if (Pawn|IsLocalPlayerController _returnvalue)
        (Audio|Components|Audio|FadeOut (Variables|Default|GetSpot_ACabin_Clock_Ticking_Int_Cue) 0.0 0.0 "Linear" :!class /Script/Engine.AudioComponent)))))

(fn GetHourHand
  (return (Variables|Living_GrandfatherClock|GetHandHour)))

(fn GetMinHand
  (return (Variables|Living_GrandfatherClock|GetHandMinute)))

(fn SetTickEnabled
  (Variables|Default|SetBShouldUpdateTime false)
  (Utilities|FlowControl|Sequence
    (:then_0
      (if (Variables|Default|GetBShouldUpdateTime)
        (Rendering|Material|SetMaterial (Variables|Living_GrandfatherClock|GetHandMinute) 0 "/Game/Environments/Props/Clocks/Materials/MI_GrandfatherClock.MI_GrandfatherClock")
        (Rendering|Material|SetMaterial (Variables|Living_GrandfatherClock|GetHandHour) 0 "/Game/Environments/Props/Clocks/Materials/MI_GrandfatherClock.MI_GrandfatherClock")
        (Transformation|SetRelativeRotation (Variables|Living_GrandfatherClock|GetHandHour) "0, 0, 0" false false)
        (else
          (Rendering|Material|SetMaterial (Variables|Living_GrandfatherClock|GetHandHour) 0 "/Game/Environments/Props/Clocks/Materials/MI_GrandfatherClock.MI_GrandfatherClock"))))
    (:then_1
      (Components|Activation|SetActive (Variables|Default|GetSpot_ACabin_ClockTicks_Singles_Cue) (Variables|Default|GetBShouldUpdateTime) false))))

(fn UpdateTime
  (Rendering|Material|SetCustomPrimitiveDataFloat (Variables|Living_GrandfatherClock|GetHandMinute) 0 (/ (TimeOfDay|GetTimeOfDay :!class /Script/FortniteGame.FortKismetLibrary) 24.0))
  (Audio|Components|Audio|Play (Variables|Default|GetSpot_ACabin_ClockTicks_Singles_Cue) 0.0))

(fn ConstructionScript
  (Actor|Tick|SetActorTickEnabled (Variables|Default|GetBShouldUpdateTime)))
