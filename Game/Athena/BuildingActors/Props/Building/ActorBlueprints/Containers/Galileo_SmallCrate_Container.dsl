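; Galileo small crate loot container Blueprint that handles opening, searched and destroyed feedback.
; On loot it plays burst VFX and sound, fades ambient audio, and animates the mesh via a timeline-driven material timestamp.
; When searched or destroyed, cosmetic instances hide glow and aura once and stop the sound indicator; non-cosmetic instances allow building over it.
; Death effects remove ambient audio, and construction adds a tag for chrome material.
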(event EventOnLoot
  ;[1 1 1] Animate Chest
  (CallFunction|SpawnChestOpenVFX)
  (bind _returnvalue (Rendering|Material|CreateDynamicMaterialInstance (Variables|Debug|GetStaticMeshComponent) 0 0 "None" :!class /Script/Engine.PrimitiveComponent))
  (bind (_direction _time) (Timeline|Timeline_0 Play)
    (:Update
      (Rendering|Material|SetScalarParameterValue _returnvalue "CODE_Timestamp" _time :!class /Script/Engine.MaterialInstanceDynamic))))

(event EventOnSetSearched
  (Networking|CanExecuteCosmeticEvents
    (:True
      (CallFunction|HideChestOpenableCosmetics)
      (SoundIndicator|StopSoundIndicatorLooping (Variables|Default|GetFortSoundIndicator)))
    (:False
      (Variables|Building|SetbDestroyOnPlayerBuildingPlacement true) ; Allows players to build over the opened chest mesh (but not when there is loot in there)
    )))

(event EventDestroyed
  (Networking|CanExecuteCosmeticEvents
    (:True
      (CallFunction|HideChestOpenableCosmetics)
      (SoundIndicator|StopSoundIndicatorLooping (Variables|Default|GetFortSoundIndicator)))
    (:False
      (Variables|Building|SetbDestroyOnPlayerBuildingPlacement true) ; Allows players to build over the opened chest mesh (but not when there is loot in there)
    )))

(event Custom|SpawnChestOpenVFX
  (Effects|SpawnEmitteratLocation (Variables|Default|GetP_Loot_Chest_Opened_FX) (Transformation|GetActorLocation) (Math|Rotator|MakeRotator (.roll (Transformation|GetActorRotation)) (+ (.pitch (Transformation|GetActorRotation)) (Variables|Default|GetOpenFX_Rotation)) (.yaw (Transformation|GetActorRotation))))
  (Audio|PlaySoundatLocation (Variables|Default|GetOpen%20Chest%20Sound) (Transformation|GetActorLocation))
  (Audio|Components|Audio|FadeOut (Variables|Default|GetAmbient%20Audio) 0.5 0.0 "Linear" :!class /Script/Engine.AudioComponent))

(event EventOnDeathPlayEffects
  (params (Damage :type Float%20%28single-precision%29) (DamageTags :type Gameplay%20Tag%20Container%20Structure%20%28by%20ref%29) (Momentum :type Vector :default "0, 0, 0") (HitInfo :type Hit%20Result%20Structure%20%28by%20ref%29) (InstigatedBy :type Fort%20Pawn%20Object%20Reference) (DamageCauser :type Actor%20Object%20Reference) (EffectContext :type Gameplay%20Effect%20Context%20Handle%20Structure))
  ;[1 1 1] Destroy Audio on Death
  (|CallParentFunction Damage DamageTags Momentum HitInfo InstigatedBy DamageCauser EffectContext)
  (Components|DestroyComponent (Variables|Default|GetAmbient%20Audio)))

(event Custom|HideChestOpenableCosmetics
  (Utilities|FlowControl|DoOnce false
    (:Completed
      (Rendering|SetVisibility (Variables|Default|GetRealmChest_Glows2) false false)
      (Components|DestroyComponent (Variables|Default|GetP_Loot_Chest_Aura_Athena)))))

(fn ConstructionScript
  ;[0.32 0.9402 1] Make sure actors that are already placed in levels get this tag so that they get the right chrome material added to them.
  (|CallParentFunction)
  (Utilities|Array|AddUnique (Variables|Actor|GetTags) "Lackluster.Material.TieredChestAthena"))
